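; Mesh terrain spline Blueprint whose construction script fits a Box component around its Spline; the event graph holds only empty events.
; It centers the box on the spline's local bounds origin, clamps X and Y extents to a minimum and Z to at least Z Distance.
; Twice Edge Width is added as padding, and the unpadded extent is debug-printed.
; Despite the asset name, this graph contains no edge noise logic.
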(event EventBeginPlay)

(event Collision|EventActorBeginOverlap
  (params (OtherActor :type Actor%20Object%20Reference)))

(event EventTick
  (params (DeltaSeconds :type Float%20%28single-precision%29)))

(fn ConstructionScript
  (bind _spline (Variables|Default|GetSpline))
  (bind _returnvalue (Fortnite|Utility|GetLocalBounds _spline))
  (bind (_origin _boxextent) (Math|BoxSphereBounds|BreakBoxSphereBounds _returnvalue))
  (bind _x (.x _boxextent))
  (bind _y (.y _boxextent))
  (bind _z (.z _boxextent))
  (bind _z_distance (Variables|Default|GetZ%20Distance))
  (bind _returnvalue_2 (Math|Float|Max(Float) _z _z_distance))
  (bind _returnvalue_3 (Math|Float|Max(Float) _y 256.0))
  (bind _returnvalue_4 (Math|Float|Max(Float) _x 256.0))
  (bind _returnvalue_5 (Math|Vector|MakeVector _returnvalue_4 _returnvalue_3 _returnvalue_2))
  (bind _box (Variables|Default|GetBox))
  (Transformation|SetRelativeLocation _box _origin false false)
  (Components|Box|SetBoxExtent _box (+ _returnvalue_5 (* (Variables|Default|GetEdge%20Width) 2.0)) true)
  (Development|PrintString (Utilities|String|ToString(Vector) _returnvalue_5)))
