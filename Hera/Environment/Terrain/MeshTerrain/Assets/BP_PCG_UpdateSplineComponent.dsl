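; PCG Blueprint element that copies a PCG spline back onto the SplineComponent of its owning BP_MeshTerrain_Road_Dirt actor.
; It writes local-space positions and tangents from the first "Spline" input into existing points by index, without adding or removing points.
; It then updates the spline and re-assigns it to the MeshPartition SplineModifier to force that modifier to refresh.
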(fn Execute
  (params (Input :type PCGData%20Collection%20Structure%20%28by%20ref%29))
  (bind _returnvalue (PCG|Helpers|GetComponentwithContext (PCG|Advanced|GetContextHandle)))
  (bind _asbp_mesh_terrain_road_dirt (Utilities|Casting|CastToBP_MeshTerrain_Road_Dirt (Components|GetOwner _returnvalue)))
  (bind (_outtaggeddata _returnvalue_2) (Data|GetTypedInputsbyPinLabel Input "Spline" "/Script/PCG.PCGSplineData"))
  (for (_array_element _array_index) (SplineData|GetSplinePoints (Utilities|Array|Get(acopy) _returnvalue_2))
    (Spline|SetLocationatSplinePoint (Actor|GetComponentbyClass _asbp_mesh_terrain_road_dirt "/Script/Engine.SplineComponent") _array_index (.Position (Utilities|Struct|BreakSplinePoint _array_element)) "Local" false)
    (Spline|SetTangentsatSplinePoint (Actor|GetComponentbyClass _asbp_mesh_terrain_road_dirt "/Script/Engine.SplineComponent") _array_index (.ArriveTangent (Utilities|Struct|BreakSplinePoint _array_element)) (.LeaveTangent (Utilities|Struct|BreakSplinePoint _array_element)) "Local" false))
  (Spline|UpdateSpline (Actor|GetComponentbyClass _asbp_mesh_terrain_road_dirt "/Script/Engine.SplineComponent"))
  ; Force Resfresh the SplineModifier
  (Spline|SetSplineComponent(Message) (Actor|GetComponentbyClass _asbp_mesh_terrain_road_dirt "/Script/MeshPartitionEditor.SplineModifier") (Actor|GetComponentbyClass _asbp_mesh_terrain_road_dirt "/Script/Engine.SplineComponent"))
  (return 0))
