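; PCG graph that generates a grass map texture array for Mesh Terrain on the GPU, one slice per actual landscape grass type (LGT).
; A Custom HLSL texture-array kernel reads channel textures, local positions, coverage mask, terrain transform, the actual-to-known LGT map and a fixed flower group mask, using shared MeshTerrain helper source.
; Width and Height inputs size the R8 output; the LGT map's element count sets its depth.
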(graph PCG_MeshTerrain_GrassMap_Generator
  (inputs (Channel%20Textures :type None) (LocalPositions :type BaseTexture) (CoverageMask :type BaseTexture) (MeshTerrainTransform :type Param) (ActualToKnownLGTMap :type Param) (Width :type Param) (Height :type Param))
  (bind GetTextureData_0 (Get%20Texture%20Data :Texture /BRCoreMaterials/Hermes/Terrain/Landscape/Textures/T_FlowerGroup01_LSGrassTypeMask.T_FlowerGroup01_LSGrassTypeMask)) ; FlowerMaskTexture
  (bind GetElementsCount_0 (Get%20Element%20Count)) ; Grass map array size should exactly match the actual LGTs
  (bind CustomHLSL_1 (Custom%20HLSL :AdditionalSources [/Hera/Environment/Terrain/MeshTerrain/Assets/ComputeSources/PCGCS_MeshTerrainHelpers.PCGCS_MeshTerrainHelpers] :InputPins [(struct :Label In :AllowedTypes (struct :Ids [(struct :Struct /Script/PCG.PCGDataTypeInfoTexture2DArray)]) :bAllowMultipleData false :PinStatus Required :bAllowMultipleConnections false) (struct :Label LocalPositions :AllowedTypes (struct :Ids [(struct :Struct /Script/PCG.PCGDataTypeInfoTexture2DSingleBase)]) :bAllowMultipleData false :PinStatus Required :bAllowMultipleConnections false) (struct :Label CoverageMask :AllowedTypes (struct :Ids [(struct :Struct /Script/PCG.PCGDataTypeInfoTexture2DSingleBase)]) :bAllowMultipleData false :PinStatus Required :bAllowMultipleConnections false) (struct :Label MeshTerrainTransform :AllowedTypes (struct :Ids [(struct :Struct /Script/PCG.PCGDataTypeInfoParam)]) :bAllowMultipleData false :PinStatus Required :bAllowMultipleConnections false) (struct :Label ActualToKnownLGT :AllowedTypes (struct :Ids [(struct :Struct /Script/PCG.PCGDataTypeInfoParam)]) :bAllowMultipleData false :PinStatus Required :bAllowMultipleConnections false) (struct :Label FlowerMask :AllowedTypes (struct :Ids [(struct :Struct /Script/PCG.PCGDataTypeInfoTexture2DSingleBase)]) :bAllowMultipleData false :PinStatus Required :bAllowMultipleConnections false)] :KernelType TextureArrayGenerator :OutputPins [(struct :PropertiesGPU (struct :bAutoInitializeOutput false) :Label Out :AllowedTypes (struct :Ids [(struct :Struct /Script/PCG.PCGDataTypeInfoTexture2DArray)]) :bAllowMultipleData false)] :TextureFormat R8))
  (connect DefaultInputNode ActualToKnownLGTMap CustomHLSL_1 ActualToKnownLGT)
  (connect DefaultInputNode CoverageMask CustomHLSL_1 CoverageMask)
  (connect GetTextureData_0 Out CustomHLSL_1 FlowerMask)
  (connect DefaultInputNode Channel%20Textures CustomHLSL_1 In)
  (connect DefaultInputNode LocalPositions CustomHLSL_1 LocalPositions)
  (connect DefaultInputNode MeshTerrainTransform CustomHLSL_1 MeshTerrainTransform)
  (connect DefaultInputNode Width CustomHLSL_1 NumElementsX)
  (connect DefaultInputNode Height CustomHLSL_1 NumElementsY)
  (connect GetElementsCount_0 Out CustomHLSL_1 NumElementsZ)
  (connect DefaultInputNode ActualToKnownLGTMap GetElementsCount_0 In)
  (output Out CustomHLSL_1))
