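; PCG graph that outputs the ordered list of landscape grass type names for MeshTerrain grass maps.
; String constants write LGT names, from lobby grass through biome types like snow, desert and redwoods, to a LandscapeGrassType attribute.
; A single Merge Attributes node combines them into the output.
; Its input pin order sets each type's position, which the graph comment says defines order in the HLSL grass map generator.
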(graph PCG_MeshTerrain_GrassMap_LGTIndices
  ;[0.15 0.15 0.15 0.5] Defines order in HLSL grass map generator
  (bind CreateConstant_0 (Create%20Constant :AttributeTypes (struct :Type String :StringValue "LGT_Grass01_Lobby" :NameValue LGT_Grass01_Lobby) :OutputTarget LandscapeGrassType))
  (bind CreateConstant_1 (Create%20Constant :AttributeTypes (struct :Type String :StringValue "LGT_Grass01" :NameValue LGT_Grass01) :OutputTarget LandscapeGrassType))
  (bind CreateConstant_2 (Create%20Constant :AttributeTypes (struct :Type String :StringValue "LGT_Gravel01" :NameValue LGT_Gravel01) :OutputTarget LandscapeGrassType))
  (bind CreateConstant_3 (Create%20Constant :AttributeTypes (struct :Type String :StringValue "LGT_Scree01" :NameValue LGT_Gravel01) :OutputTarget LandscapeGrassType))
  (bind CreateConstant_4 (Create%20Constant :AttributeTypes (struct :Type String :StringValue "LGT_Snow01" :NameValue LGT_Gravel01) :OutputTarget LandscapeGrassType))
  (bind CreateConstant_5 (Create%20Constant :AttributeTypes (struct :Type String :StringValue "LGT_ForestFloor" :NameValue LGT_Gravel01) :OutputTarget LandscapeGrassType))
  (bind CreateConstant_6 (Create%20Constant :AttributeTypes (struct :Type String :StringValue "LGT_FlowerGroup01" :NameValue LGT_Gravel01) :OutputTarget LandscapeGrassType))
  (bind CreateConstant_7 (Create%20Constant :AttributeTypes (struct :Type String :StringValue "LGT_ChaparralGrass" :NameValue LGT_Gravel01) :OutputTarget LandscapeGrassType))
  (bind CreateConstant_8 (Create%20Constant :AttributeTypes (struct :Type String :StringValue "LGT_DesertYucca" :NameValue LGT_Gravel01) :OutputTarget LandscapeGrassType))
  (bind CreateConstant_9 (Create%20Constant :AttributeTypes (struct :Type String :StringValue "LGT_DesertSand" :NameValue LGT_Gravel01) :OutputTarget LandscapeGrassType))
  (bind CreateConstant_10 (Create%20Constant :AttributeTypes (struct :Type String :StringValue "LGT_Beach" :NameValue LGT_Gravel01) :OutputTarget LandscapeGrassType))
  (bind CreateConstant_11 (Create%20Constant :AttributeTypes (struct :Type String :StringValue "LGT_RedWoods" :NameValue LGT_Gravel01) :OutputTarget LandscapeGrassType))
  (bind CreateConstant_12 (Create%20Constant :AttributeTypes (struct :Type String :StringValue "LGT_Archipelago" :NameValue LGT_Gravel01) :OutputTarget LandscapeGrassType))
  (bind CreateConstant_13 (Create%20Constant :AttributeTypes (struct :Type String :StringValue "LGT_Grass_Seaweed_A" :NameValue LGT_Gravel01) :OutputTarget LandscapeGrassType))
  (bind CreateConstant_14 (Create%20Constant :AttributeTypes (struct :Type String :StringValue "LGT_Grass_GoldenFall" :NameValue LGT_Gravel01) :OutputTarget LandscapeGrassType))
  (bind MergeAttributes_0 (Merge%20Attributes))
  (connect CreateConstant_0 Out MergeAttributes_0 In)
  (connect CreateConstant_9 Out MergeAttributes_0 In10)
  (connect CreateConstant_10 Out MergeAttributes_0 In11)
  (connect CreateConstant_11 Out MergeAttributes_0 In12)
  (connect CreateConstant_12 Out MergeAttributes_0 In13)
  (connect CreateConstant_13 Out MergeAttributes_0 In14)
  (connect CreateConstant_14 Out MergeAttributes_0 In15)
  (connect CreateConstant_1 Out MergeAttributes_0 In2)
  (connect CreateConstant_2 Out MergeAttributes_0 In3)
  (connect CreateConstant_3 Out MergeAttributes_0 In4)
  (connect CreateConstant_4 Out MergeAttributes_0 In5)
  (connect CreateConstant_5 Out MergeAttributes_0 In6)
  (connect CreateConstant_6 Out MergeAttributes_0 In7)
  (connect CreateConstant_7 Out MergeAttributes_0 In8)
  (connect CreateConstant_8 Out MergeAttributes_0 In9)
  (output Out MergeAttributes_0))
