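; PCG graph that drives grass generation on the Hera mesh terrain from Landscape Grass Types.
; An unbounded-grid pass turns the mesh partition's grass types into layer, primitive and type data, plus texel sizes and an actual-to-known type map.
; Custom HLSL nodes upload layer and texel data early for downstream GPU use.
; On a 128 grid matching World Partition sections, each terrain section feeds the GeneratorLoop subgraph with that shared data.
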(graph PCG_MeshTerrain_Grass_Generator
  ;[0.15 0.15 0.15 0.5] Preprocessing - Runs on the Unbounded Grid
  ;;[0.15 0.15 0.15 0.5] Process Landscape Grass Types (LGTs)
  (bind GetMeshPartitionGrassTypes_1 (Get%20Mesh%20Partition%20Grass%20Types :Definition /Hera/Environment/Terrain/MeshTerrain/MPD_Hera.MPD_Hera))
  (bind ProcessGrassTypes_0 (Process%20Grass%20Types))
  (bind RemoveEmptyData_1 (Remove%20Empty%20Data))
  (bind AttributePartition_0 (Attribute%20Partition :PartitionAttributeSelectors [PrimitiveIndex]))
  (bind GetAttributeFromIndex_0 (Get%20Attribute%20Set%20from%20Index))
  (bind MergeAttributes_0 (Merge%20Attributes))
  (bind DeleteAttributes_0 (Delete%20Attributes :Operation KeepSelectedAttributes :SelectedAttributes "SumDensities,PrimitiveIndex,ScaleMin,ScaleMax,ScalingMode,ScaleByWeight,RandomRotation,AlignToSurface"))
  (bind AttributeMathsOp_0 (Attribute%20Maths%20Op :InputSource1 SumDensities :Operation Multiply)) ; Density conversion
  (bind CustomHLSL_1 (Custom%20HLSL :InputPins [(struct :Label In :AllowedTypes (struct :Ids [(struct :Struct /Script/PCG.PCGDataTypeInfoParam)]) :PinStatus Required :bAllowMultipleConnections false)] :KernelType AttributeSetProcessor :OutputPins [(struct :PropertiesGPU (struct :PinsToInititalizeFrom [In]) :Label Out :AllowedTypes (struct :Ids [(struct :Struct /Script/PCG.PCGDataTypeInfoParam)]))])) ; Upload at the UB level since we use on the GPU downstream
  (bind DeleteAttributes_1 (Delete%20Attributes :Operation KeepSelectedAttributes :SelectedAttributes "LandscapeGrassType"))
  (bind GetAttributeFromIndex_2 (Get%20Attribute%20Set%20from%20Index))
  (bind MergeAttributes_1 (Merge%20Attributes))
  ;;[0.15 0.15 0.15 0.5] Get Channel Textures Texel Size
  (bind GetMeshPartitionTexelSizes_1 (Get%20Mesh%20Partition%20Texel%20Sizes :Definition /Hera/Environment/Terrain/MeshTerrain/MPD_Hera.MPD_Hera))
  (bind CustomHLSL_10 (Custom%20HLSL :InputPins [(struct :Label In :AllowedTypes (struct :Ids [(struct :Struct /Script/PCG.PCGDataTypeInfoParam)]) :PinStatus Required :bAllowMultipleConnections false)] :KernelType AttributeSetProcessor :OutputPins [(struct :PropertiesGPU (struct :PinsToInititalizeFrom [In]) :Label Out :AllowedTypes (struct :Ids [(struct :Struct /Script/PCG.PCGDataTypeInfoParam)]))]))
  ;;[0.15 0.15 0.15 0.5] Build map from Actual LGTs to Known LGTs
  (bind Subgraph_1 (Subgraph :Graph /Hera/Environment/Terrain/MeshTerrain/Assets/PCG_MeshTerrain_GrassMap_LGTIndices.PCG_MeshTerrain_GrassMap_LGTIndices))
  (bind CustomHLSL_2 (Custom%20HLSL :InputPins [(struct :Label ActualLGTs :AllowedTypes (struct :Ids [(struct :Struct /Script/PCG.PCGDataTypeInfoParam)]) :bAllowMultipleData false :PinStatus Required :bAllowMultipleConnections false) (struct :Label KnownLGTs :AllowedTypes (struct :Ids [(struct :Struct /Script/PCG.PCGDataTypeInfoParam)]) :bAllowMultipleData false :PinStatus Required :bAllowMultipleConnections false)] :KernelType Custom :OutputPins [(struct :PropertiesGPU (struct :PinsToInititalizeFrom [ActualLGTs] :AttributeInheritanceMode None) :Label Out :AllowedTypes (struct :Ids [(struct :Struct /Script/PCG.PCGDataTypeInfoParam)]) :bAllowMultipleData false)]))
  ;[0.15 0.15 0.15 0.5] Per grid cell
  (bind HiGenGridSize_2 (Change%20Grid%20Size :HiGenGridSize Grid128)) ; Set grid to match the mesh terrain section WP grid
  (bind GetMeshTerrainSection_2 (Get%20Mesh%20Terrain%20Section :bEmitMultipleSections false :SelectedBuildVariants [Common]))
  (bind Subgraph_0 (Subgraph :Graph /Hera/Environment/Terrain/MeshTerrain/Assets/PCG_MeshTerrain_Grass_GeneratorLoop.PCG_MeshTerrain_Grass_GeneratorLoop))
  (reroute ActualToKnownLGTs CustomHLSL_2)
  (reroute GrassTypes MergeAttributes_1)
  (reroute LayerData CustomHLSL_1)
  (reroute PrimData MergeAttributes_0)
  (reroute TexelSize CustomHLSL_10)
  (connect DeleteAttributes_0 Out AttributeMathsOp_0 InA)
  (connect RemoveEmptyData_1 Out AttributePartition_0 In)
  (connect AttributeMathsOp_0 Out CustomHLSL_1 In)
  (connect GrassTypes Out CustomHLSL_2 ActualLGTs)
  (connect Subgraph_1 Out CustomHLSL_2 KnownLGTs)
  (connect GetMeshPartitionTexelSizes_1 Out CustomHLSL_10 In)
  (connect RemoveEmptyData_1 Out DeleteAttributes_0 In)
  (connect RemoveEmptyData_1 Out DeleteAttributes_1 In)
  (connect AttributePartition_0 Out GetAttributeFromIndex_0 In)
  (connect DeleteAttributes_1 Out GetAttributeFromIndex_2 In)
  (connect HiGenGridSize_2 Execution%20Dependency GetMeshTerrainSection_2 Execution%20Dependency)
  (connect GetAttributeFromIndex_0 Out MergeAttributes_0 In)
  (connect GetAttributeFromIndex_2 Out MergeAttributes_1 In)
  (connect GetMeshPartitionGrassTypes_1 Out ProcessGrassTypes_0 In)
  (connect ProcessGrassTypes_0 Out RemoveEmptyData_1 In)
  (connect GetMeshTerrainSection_2 Out Subgraph_0 In)
  (connect ActualToKnownLGTs Out Subgraph_0 InActualToKnownLGTs)
  (connect GrassTypes Out Subgraph_0 InGrassTypes)
  (connect LayerData Out Subgraph_0 InLayerData)
  (connect PrimData Out Subgraph_0 InPrimData)
  (connect TexelSize Out Subgraph_0 InTexelSIze))
