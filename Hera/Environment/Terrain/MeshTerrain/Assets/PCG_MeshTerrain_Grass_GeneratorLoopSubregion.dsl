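; PCG loop subgraph that scatters GPU-instanced grass over a Mesh Terrain section.
; It bakes section positions and coverage into UV space, then crops them and the hard-coded material/biome channels to computed bounds.
; A subgraph derives grass maps, whose max-downsampled mask and layer data yield group seeds and thread counts.
; An HLSL point generator feeds a primitive-table mesh spawner; partly wired debug branches visualize intermediates.
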(graph PCG_MeshTerrain_Grass_GeneratorLoopSubregion
  (inputs (InPrimData :type Param) (InLayerData :type Param) (InGrassTypes :type Param) (InActualToKnownLGTs :type Param) (InTexelSIze :type Param))
  ;[0.15 0.15 0.15 0.5] Per grid
  ;;[0.15 0.15 0.15 0.5] Process Mesh Terrain Section
  ;;;[0.15 0.15 0.15 0.5] Read mesh section
  (bind CreateConstant_1 (Create%20Constant :AttributeTypes (struct :Type Transform) :OutputTarget Origin))
  (bind GetMeshTerrainSectionActor_0 (Get%20Mesh%20Terrain%20Section%20Actor))
  (bind GetActorData_0 (Get%20Actor%20Data :ActorSelector (struct :ActorFilter FromInput) :Mode GetSinglePoint))
  (bind CopyAttributes_1 (Copy%20Attributes :InputSource $Transform :OutputTarget MeshTerrainTransform))
  (bind CustomHLSL_1 (Custom%20HLSL :InputPins [(struct :Label In :AllowedTypes (struct :Ids [(struct :Struct /Script/PCG.PCGDataTypeInfoParam)]) :PinStatus Required :bAllowMultipleConnections false)] :KernelType AttributeSetProcessor :OutputPins [(struct :PropertiesGPU (struct :PinsToInititalizeFrom [In]) :Label Out :AllowedTypes (struct :Ids [(struct :Struct /Script/PCG.PCGDataTypeInfoParam)]))]))
  (bind GetMeshTerrainSectionChannelTextures_1 (Get%20Mesh%20Terrain%20Section%20Channel%20Textures :bExcludeSelectedChannels false :SelectedChannels [Mat_Gravel Mat_Mud Mat_Sand Mat_Asphalt Biome_A Biome_B Biome_C Biome_D])) ; Grab the channel textures, selected channels are hard-coded here
  (bind GetTextureInfo_0 (Get%20Texture%20Info :bOutputComputeFormat false :bOutputComputeFormatIndex false :bOutputFormat false :bOutputFormatIndex false :bOutputNumMips false))
  ;;;[0.15 0.15 0.15 0.5] Extract relevant part of mesh section data
  (bind GetTextureInfo_1 (Get%20Texture%20Info :bOutputComputeFormat false :bOutputComputeFormatIndex false :bOutputFormat false :bOutputFormatIndex false :bOutputNumMips false))
  (bind ExtractAttributeAtIndex_0 (Extract%20Attribute%20at%20Index :bForceOutputAttributeToBeInElementsDomain true :InputSource Width))
  (bind ExtractAttributeAtIndex_1 (Extract%20Attribute%20at%20Index :bForceOutputAttributeToBeInElementsDomain true :InputSource Height))
  (bind CustomHLSL_19 (Custom%20HLSL :AdditionalSources [/Hera/Environment/Terrain/MeshTerrain/Assets/ComputeSources/PCGCS_MeshTerrainHelpers.PCGCS_MeshTerrainHelpers] :bMuteUnwrittenPinDataErrors true :DispatchThreadCount FromProductOfInputPins :InputPins [(struct :Label LocalPositions :AllowedTypes (struct :Ids [(struct :Struct /Script/PCG.PCGDataTypeInfoTexture2DSingleBase)]) :bAllowMultipleData false :PinStatus Required :bAllowMultipleConnections false) (struct :Label MeshTerrainTransform :AllowedTypes (struct :Ids [(struct :Struct /Script/PCG.PCGDataTypeInfoParam)]) :bAllowMultipleData false :bAllowMultipleConnections false) (struct :Label TexelSize :AllowedTypes (struct :Ids [(struct :Struct /Script/PCG.PCGDataTypeInfoParam)]) :bAllowMultipleData false :bAllowMultipleConnections false)] :KernelType Custom :OutputPins [(struct :PropertiesGPU (struct :InitializationMode Custom :PinsToInititalizeFrom [In] :bInitializeToZero true) :Label Bounds :AllowedTypes (struct :Ids [(struct :Struct /Script/PCG.PCGDataTypeInfoParam)]) :bAllowMultipleData false)] :ThreadCountInputPinLabels [LocalPositions]))
  (bind CustomHLSL_26 (Custom%20HLSL :bMuteUnwrittenPinDataErrors true :InputPins [(struct :Label In :AllowedTypes (struct :Ids [(struct :Struct /Script/PCG.PCGDataTypeInfoParam)]) :PinStatus Required :bAllowMultipleConnections false)] :KernelType Custom :OutputPins [(struct :PropertiesGPU (struct :PinsToInititalizeFrom [In] :bAutoInitializeOutput false :AttributeInheritanceMode None) :Label MinX :AllowedTypes (struct :Ids [(struct :Struct /Script/PCG.PCGDataTypeInfoParam)]) :bAllowMultipleData false) (struct :PropertiesGPU (struct :PinsToInititalizeFrom [In] :bAutoInitializeOutput false :AttributeInheritanceMode None) :Label MinY :AllowedTypes (struct :Ids [(struct :Struct /Script/PCG.PCGDataTypeInfoParam)]) :bAllowMultipleData false) (struct :PropertiesGPU (struct :PinsToInititalizeFrom [In] :bAutoInitializeOutput false :AttributeInheritanceMode None) :Label ResX :AllowedTypes (struct :Ids [(struct :Struct /Script/PCG.PCGDataTypeInfoParam)]) :bAllowMultipleData false) (struct :PropertiesGPU (struct :PinsToInititalizeFrom [In] :bAutoInitializeOutput false :AttributeInheritanceMode None) :Label ResY :AllowedTypes (struct :Ids [(struct :Struct /Script/PCG.PCGDataTypeInfoParam)]) :bAllowMultipleData false)]))
  (bind CustomHLSL_20 (Custom%20HLSL :InputPins [(struct :Label LocalPositions :AllowedTypes (struct :Ids [(struct :Struct /Script/PCG.PCGDataTypeInfoTexture2DSingleBase)]) :bAllowMultipleData false :PinStatus Required :bAllowMultipleConnections false) (struct :Label MinX :AllowedTypes (struct :Ids [(struct :Struct /Script/PCG.PCGDataTypeInfoParam)]) :bAllowMultipleData false :bAllowMultipleConnections false) (struct :Label MinY :AllowedTypes (struct :Ids [(struct :Struct /Script/PCG.PCGDataTypeInfoParam)]) :bAllowMultipleData false :bAllowMultipleConnections false)] :KernelType TextureGenerator :OutputPins [(struct :PropertiesGPU (struct :bAutoInitializeOutput false) :Label Out :AllowedTypes (struct :Ids [(struct :Struct /Script/PCG.PCGDataTypeInfoTexture2DSingleBase)]) :bAllowMultipleData false)]))
  (bind CustomHLSL_21 (Custom%20HLSL :InputPins [(struct :Label LocalPositions :AllowedTypes (struct :Ids [(struct :Struct /Script/PCG.PCGDataTypeInfoTexture2DSingleBase)]) :bAllowMultipleData false :PinStatus Required :bAllowMultipleConnections false) (struct :Label MinX :AllowedTypes (struct :Ids [(struct :Struct /Script/PCG.PCGDataTypeInfoParam)]) :bAllowMultipleData false :bAllowMultipleConnections false) (struct :Label MinY :AllowedTypes (struct :Ids [(struct :Struct /Script/PCG.PCGDataTypeInfoParam)]) :bAllowMultipleData false :bAllowMultipleConnections false)] :KernelType TextureGenerator :OutputPins [(struct :PropertiesGPU (struct :bAutoInitializeOutput false) :Label Out :AllowedTypes (struct :Ids [(struct :Struct /Script/PCG.PCGDataTypeInfoTexture2DSingleBase)]) :bAllowMultipleData false)] :TextureFormat R8))
  (bind ExtractAttributeAtIndex_3 (Extract%20Attribute%20at%20Index :bForceOutputAttributeToBeInElementsDomain true :InputSource ArraySize))
  (bind CustomHLSL_22 (Custom%20HLSL :InputPins [(struct :Label In :AllowedTypes (struct :Ids [(struct :Struct /Script/PCG.PCGDataTypeInfoTexture2DArray)]) :bAllowMultipleData false :PinStatus Required :bAllowMultipleConnections false) (struct :Label MinX :AllowedTypes (struct :Ids [(struct :Struct /Script/PCG.PCGDataTypeInfoParam)]) :bAllowMultipleData false :bAllowMultipleConnections false) (struct :Label MinY :AllowedTypes (struct :Ids [(struct :Struct /Script/PCG.PCGDataTypeInfoParam)]) :bAllowMultipleData false :bAllowMultipleConnections false)] :KernelType TextureArrayGenerator :OutputPins [(struct :PropertiesGPU (struct) :Label Out :AllowedTypes (struct :Ids [(struct :Struct /Script/PCG.PCGDataTypeInfoTexture2DArray)]) :bAllowMultipleData false)]))
  ;;;[0.15 0.15 0.15 0.5] Unwrap static mesh attributes into UV space
  (bind ExtractAttributeAtIndex_4 (Extract%20Attribute%20at%20Index :bForceOutputAttributeToBeInElementsDomain true :InputSource Width))
  (bind ExtractAttributeAtIndex_5 (Extract%20Attribute%20at%20Index :bForceOutputAttributeToBeInElementsDomain true :InputSource Height))
  (bind BakeMeshTerrainSectionMesh_1 (Bake%20Mesh%20Terrain%20Section%20Mesh))
  (bind BakeMeshTerrainSectionMesh_2 (Bake%20Mesh%20Terrain%20Section%20Mesh :BakeParams (struct :Format RTF_R8 :Attribute Mask :Padding 0)))
  ;;[0.15 0.15 0.15 0.5] Process Grass Maps
  (bind CustomHLSL_2 (Custom%20HLSL :bMuteUnwrittenPinDataErrors true :InputPins [(struct :Label In :AllowedTypes (struct :Ids [(struct :Struct /Script/PCG.PCGDataTypeInfoTexture2DArray)]) :bAllowMultipleData false :PinStatus Required :bAllowMultipleConnections false)] :KernelSourceOverride /Game/Athena/Helios/Environments/Landscape/WorldBuildingTools/PCGGrass/Resources/ComputeSources/PCGCS_GenerateMask.PCGCS_GenerateMask :KernelType Custom :OutputPins [(struct :PropertiesGPU (struct :PinsToInititalizeFrom [In] :TextureFormat R32f) :Label Out :AllowedTypes (struct :Ids [(struct :Struct /Script/PCG.PCGDataTypeInfoTexture2DArray)]) :bAllowMultipleData false)]))
  (bind DownsampleTexture_0 (Downsample%20Texture :Mode Max))
  ;;[0.15 0.15 0.15 0.5] Scattering Prep
  (bind CustomHLSL_3 (Custom%20HLSL :InputPins [(struct :Label GrassMaps :AllowedTypes (struct :Ids [(struct :Struct /Script/PCG.PCGDataTypeInfoTexture2DArray)]) :bAllowMultipleData false :bAllowMultipleConnections false)] :KernelType AttributeSetGenerator :NumElements 1 :OutputPins [(struct :PropertiesGPU (struct :CreatedKernelAttributeKeys [(struct :Identifier (struct :Name NumSlices) :Type Int :Name NumSlices)]) :Label NumSlices :AllowedTypes (struct :Ids [(struct :Struct /Script/PCG.PCGDataTypeInfoParam)]) :bAllowMultipleData false)]))
  (bind CustomHLSL_4 (Custom%20HLSL :InputPins [(struct :Label Mask :AllowedTypes (struct :Ids [(struct :Struct /Script/PCG.PCGDataTypeInfoTexture2DArray)]) :bAllowMultipleData false :bAllowMultipleConnections false) (struct :Label LayerData :AllowedTypes (struct :Ids [(struct :Struct /Script/PCG.PCGDataTypeInfoParam)]) :bAllowMultipleConnections false) (struct :Label TexelSize :AllowedTypes (struct :Ids [(struct :Struct /Script/PCG.PCGDataTypeInfoParam)]) :bAllowMultipleData false :bAllowMultipleConnections false) (struct :Label GrassMaps :AllowedTypes (struct :Ids [(struct :Struct /Script/PCG.PCGDataTypeInfoTexture2DArray)]) :bAllowMultipleData false :bAllowMultipleConnections false)] :KernelSourceOverride /Hera/Environment/Terrain/MeshTerrain/Assets/ComputeSources/PCGCS_LayerInfo_MT.PCGCS_LayerInfo_MT :KernelType AttributeSetGenerator :OutputPins [(struct :PropertiesGPU (struct :InitializationMode Custom :PinsToInititalizeFrom [GrassMaps]) :Label LayerInfo :AllowedTypes (struct :Ids [(struct :Struct /Script/PCG.PCGDataTypeInfoParam)]) :bAllowMultipleData false) (struct :PropertiesGPU (struct :InitializationMode Custom :PinsToInititalizeFrom [GrassMaps] :bInitializeToZero true :CreatedKernelAttributeKeys [(struct :Identifier (struct :Name TotalMaxNumCells) :Type Int :Name TotalMaxNumCells)]) :Label TotalMaxNumCells :AllowedTypes (struct :Ids [(struct :Struct /Script/PCG.PCGDataTypeInfoParam)]) :bAllowMultipleData false)]))
  (bind CustomHLSL_5 (Custom%20HLSL :InputPins [(struct :Label Mask :AllowedTypes (struct :Ids [(struct :Struct /Script/PCG.PCGDataTypeInfoTexture2DArray)]) :bAllowMultipleData false :bAllowMultipleConnections false) (struct :Label LayerInfo :AllowedTypes (struct :Ids [(struct :Struct /Script/PCG.PCGDataTypeInfoParam)]) :bAllowMultipleData false :PinStatus Required :bAllowMultipleConnections false) (struct :Label TexelSize :AllowedTypes (struct :Ids [(struct :Struct /Script/PCG.PCGDataTypeInfoParam)]) :bAllowMultipleData false :bAllowMultipleConnections false) (struct :Label GrassMaps :AllowedTypes (struct :Ids [(struct :Struct /Script/PCG.PCGDataTypeInfoTexture2DArray)]) :bAllowMultipleData false :bAllowMultipleConnections false)] :KernelSourceOverride /Hera/Environment/Terrain/MeshTerrain/Assets/ComputeSources/PCGCS_CreateGroupSeeds_MT.PCGCS_CreateGroupSeeds_MT :KernelType AttributeSetGenerator :OutputPins [(struct :PropertiesGPU (struct :PinsToInititalizeFrom [LayerInfo] :ElementCount 256 :bInitializeToZero true :bAutoInitializeOutput false :AttributeInheritanceMode None) :Label OutSeeds :AllowedTypes (struct :Ids [(struct :Struct /Script/PCG.PCGDataTypeInfoParam)]) :bAllowMultipleData false) (struct :PropertiesGPU (struct :InitializationMode Custom :PinsToInititalizeFrom [LayerInfo] :bInitializeToZero true :CreatedKernelAttributeKeys [(struct :Identifier (struct :Name NumThreads) :Type Int :Name NumThreads)]) :Label NumThreads :AllowedTypes (struct :Ids [(struct :Struct /Script/PCG.PCGDataTypeInfoParam)]) :bAllowMultipleData false)]))
  ;;[0.15 0.15 0.15 0.5] Scattering
  (bind CustomHLSL_6 (Custom%20HLSL :AdditionalSources [/Game/Athena/Helios/Environments/Landscape/WorldBuildingTools/PCGGrass/Resources/ComputeSources/PCGCS_InstanceGrass_Functions.PCGCS_InstanceGrass_Functions /Hera/Environment/Terrain/MeshTerrain/Assets/ComputeSources/PCGCS_MeshTerrainHelpers.PCGCS_MeshTerrainHelpers] :InputPins [(struct :Label Seeds :AllowedTypes (struct :Ids [(struct :Struct /Script/PCG.PCGDataTypeInfoParam)]) :bAllowMultipleData false :PinStatus Required :bAllowMultipleConnections false) (struct :Label GrassMap :AllowedTypes (struct :Ids [(struct :Struct /Script/PCG.PCGDataTypeInfoTexture2DArray)]) :bAllowMultipleData false :bAllowMultipleConnections false) (struct :Label LocalPositions :AllowedTypes (struct :Ids [(struct :Struct /Script/PCG.PCGDataTypeInfoTexture2DSingleBase)]) :bAllowMultipleData false :bAllowMultipleConnections false) (struct :Label LayerData :AllowedTypes (struct :Ids [(struct :Struct /Script/PCG.PCGDataTypeInfoParam)]) :bAllowMultipleConnections false) (struct :Label MeshTerrainTransform :AllowedTypes (struct :Ids [(struct :Struct /Script/PCG.PCGDataTypeInfoParam)]) :bAllowMultipleData false :bAllowMultipleConnections false) (struct :Label TexelSize :AllowedTypes (struct :Ids [(struct :Struct /Script/PCG.PCGDataTypeInfoParam)]) :bAllowMultipleData false :bAllowMultipleConnections false) (struct :Label LayerInfo :AllowedTypes (struct :Ids [(struct :Struct /Script/PCG.PCGDataTypeInfoParam)]) :bAllowMultipleData false :PinStatus Required :bAllowMultipleConnections false) (struct :Label Mask :AllowedTypes (struct :Ids [(struct :Struct /Script/PCG.PCGDataTypeInfoTexture2DSingleBase)]) :bAllowMultipleData false :bAllowMultipleConnections false)] :KernelSourceOverride /Hera/Environment/Terrain/MeshTerrain/Assets/ComputeSources/PCGCS_InstanceGrass_MT.PCGCS_InstanceGrass_MT :KernelType PointGenerator :OutputPins [(struct :PropertiesGPU (struct :bAutoInitializeOutput false) :Label Out :AllowedTypes (struct :Ids [(struct :Struct /Script/PCG.PCGDataTypeInfoPoint)]))]))
  (bind StaticMeshSpawner_1 (Static%20Mesh%20Spawner :bApplyMeshBoundsToPoints false :MeshSelectorParameters (class:/Script/PCG.PCGMeshSelectorPrimitiveData :MaterialOverrideAttributes [OverrideMaterial0 OverrideMaterial1] :MeshAttribute MeshAsset :PrimitiveIndexAttribute PrimitiveIndex :PrimitiveOverrideAttributes [(struct :InputSource StartCullDistance :PropertyTarget "InstanceStartCullDistance") (struct :InputSource EndCullDistance :PropertyTarget "InstanceEndCullDistance")] :TemplateDescriptor (struct :InstanceStartCullDistance 2500 :InstanceEndCullDistance 3000 :bCastShadow false :bAffectDistanceFieldLighting false :bReceivesDecals false :bVisibleInRayTracing false)) :MeshSelectorType /Script/PCG.PCGMeshSelectorPrimitiveData))
  ;;[0.15 0.15 0.15 0.5] Compute Grass Maps from Weight Channel Textures
  (bind Subgraph_1 (Subgraph :Graph /Hera/Environment/Terrain/MeshTerrain/Assets/PCG_MeshTerrain_GrassMap_Generator.PCG_MeshTerrain_GrassMap_Generator))
  ;[0.4375 0.1663 0.3504 0.5565] Debug
  ;;[0.4375 0 0.2969 0.5] Debug Index
  (bind CreateConstant_3 (Create%20Constant :AttributeTypes (struct :Type Integer32) :OutputTarget SliceIndex))
  ;;[0.4375 0 0.2969 0.5] Debug Group Seed (world)
  (bind CustomHLSL_7 (Custom%20HLSL :AdditionalSources [/Hera/Environment/Terrain/MeshTerrain/Assets/ComputeSources/PCGCS_MeshTerrainHelpers.PCGCS_MeshTerrainHelpers] :InputPins [(struct :Label Seeds :AllowedTypes (struct :Ids [(struct :Struct /Script/PCG.PCGDataTypeInfoParam)]) :bAllowMultipleData false :PinStatus Required :bAllowMultipleConnections false) (struct :Label Layer :AllowedTypes (struct :Ids [(struct :Struct /Script/PCG.PCGDataTypeInfoParam)]) :bAllowMultipleData false :bAllowMultipleConnections false) (struct :Label LocalPositions :AllowedTypes (struct :Ids [(struct :Struct /Script/PCG.PCGDataTypeInfoTexture2DSingleBase)]) :bAllowMultipleData false :bAllowMultipleConnections false) (struct :Label MeshTerrainTransform :AllowedTypes (struct :Ids [(struct :Struct /Script/PCG.PCGDataTypeInfoParam)]) :bAllowMultipleData false :bAllowMultipleConnections false)] :KernelType Custom :OutputPins [(struct :PropertiesGPU (struct :PinsToInititalizeFrom [Seeds] :AttributeInheritanceMode None) :Label Out :AllowedTypes (struct :Ids [(struct :Struct /Script/PCG.PCGDataTypeInfoPoint)]) :bAllowMultipleData false)]))
  (bind CustomHLSL_8 (Custom%20HLSL :OutputPins [(struct :PropertiesGPU (struct :PinsToInititalizeFrom [In] :CreatedKernelAttributeKeys [(struct :Identifier (struct :Name MyDensity) :Name MyDensity)]) :Label Out :AllowedTypes (struct :Ids [(struct :Struct /Script/PCG.PCGDataTypeInfoPoint)]))]))
  (bind StaticMeshSpawner_2 (Static%20Mesh%20Spawner :InstanceDataPackerParameters (class:/Script/PCG.PCGInstanceDataPackerByAttribute :AttributeSelectors [MyDensity]) :InstanceDataPackerType /Script/PCG.PCGInstanceDataPackerByAttribute :MeshSelectorParameters (class:/Script/PCG.PCGMeshSelectorWeighted :MeshEntries [(struct :Descriptor (struct :StaticMesh /PCG/DebugObjects/PCG_CubeWireframe.PCG_CubeWireframe :OverrideMaterials [/PCG/DebugObjects/PCG_DebugMaterial.PCG_DebugMaterial]))])))
  ;;[0.4375 0 0.2969 0.5] Debug Grass Maps (world)
  (bind CustomHLSL_25 (Custom%20HLSL :AdditionalSources [/Hera/Environment/Terrain/MeshTerrain/Assets/ComputeSources/PCGCS_MeshTerrainHelpers.PCGCS_MeshTerrainHelpers] :bMuteUnwrittenPinDataErrors true :InputPins [(struct :Label In :AllowedTypes (struct :Ids [(struct :Struct /Script/PCG.PCGDataTypeInfoTexture2DArray)]) :bAllowMultipleData false :PinStatus Required :bAllowMultipleConnections false) (struct :Label LocalPositions :AllowedTypes (struct :Ids [(struct :Struct /Script/PCG.PCGDataTypeInfoTexture2DSingleBase)]) :bAllowMultipleData false :bAllowMultipleConnections false) (struct :Label MeshTerrainTransform :AllowedTypes (struct :Ids [(struct :Struct /Script/PCG.PCGDataTypeInfoParam)]) :bAllowMultipleData false :bAllowMultipleConnections false) (struct :Label SliceIndex :AllowedTypes (struct :Ids [(struct :Struct /Script/PCG.PCGDataTypeInfoParam)]) :bAllowMultipleData false :bAllowMultipleConnections false)] :KernelType Custom :OutputPins [(struct :PropertiesGPU (struct :PinsToInititalizeFrom [In]) :Label Out :AllowedTypes (struct :Ids [(struct :Struct /Script/PCG.PCGDataTypeInfoPoint)]) :bAllowMultipleData false)]))
  (bind CustomHLSL_11 (Custom%20HLSL :OutputPins [(struct :PropertiesGPU (struct :PinsToInititalizeFrom [In]) :Label Out :AllowedTypes (struct :Ids [(struct :Struct /Script/PCG.PCGDataTypeInfoPoint)]))]))
  (bind StaticMeshSpawner_3 (Static%20Mesh%20Spawner :InstanceDataPackerParameters (class:/Script/PCG.PCGInstanceDataPackerByAttribute :AttributeSelectors [MyDensity MyExtents MyColor]) :InstanceDataPackerType /Script/PCG.PCGInstanceDataPackerByAttribute :MeshSelectorParameters (class:/Script/PCG.PCGMeshSelectorWeighted :MeshEntries [(struct :Descriptor (struct :StaticMesh /PCG/DebugObjects/PCG_Cube.PCG_Cube :OverrideMaterials [/PCG/DebugObjects/PCG_DebugMaterial.PCG_DebugMaterial]))])))
  ;;[0.4375 0 0.2969 0.5] Debug Positions (world)
  (bind CustomHLSL_12 (Custom%20HLSL :AdditionalSources [/Hera/Environment/Terrain/MeshTerrain/Assets/ComputeSources/PCGCS_MeshTerrainHelpers.PCGCS_MeshTerrainHelpers] :bMuteUnwrittenPinDataErrors true :InputPins [(struct :Label LocalPositions :AllowedTypes (struct :Ids [(struct :Struct /Script/PCG.PCGDataTypeInfoTexture2DSingleBase)]) :bAllowMultipleData false :bAllowMultipleConnections false) (struct :Label MeshTerrainTransform :AllowedTypes (struct :Ids [(struct :Struct /Script/PCG.PCGDataTypeInfoParam)]) :bAllowMultipleData false :bAllowMultipleConnections false)] :KernelType Custom :OutputPins [(struct :PropertiesGPU (struct :PinsToInititalizeFrom [LocalPositions]) :Label Out :AllowedTypes (struct :Ids [(struct :Struct /Script/PCG.PCGDataTypeInfoPoint)]) :bAllowMultipleData false)]))
  (bind CustomHLSL_13 (Custom%20HLSL :OutputPins [(struct :PropertiesGPU (struct :PinsToInititalizeFrom [In]) :Label Out :AllowedTypes (struct :Ids [(struct :Struct /Script/PCG.PCGDataTypeInfoPoint)]))]))
  (bind StaticMeshSpawner_4 (Static%20Mesh%20Spawner :InstanceDataPackerParameters (class:/Script/PCG.PCGInstanceDataPackerByAttribute :AttributeSelectors [MyDensity MyExtents MyColor]) :InstanceDataPackerType /Script/PCG.PCGInstanceDataPackerByAttribute :MeshSelectorParameters (class:/Script/PCG.PCGMeshSelectorWeighted :MeshEntries [(struct :Descriptor (struct :StaticMesh /PCG/DebugObjects/PCG_Cube.PCG_Cube :OverrideMaterials [/PCG/DebugObjects/PCG_DebugMaterial_Color.PCG_DebugMaterial_Color]))])))
  ;;[0.4375 0 0.2969 0.5] Debug Positions (viewport)
  (bind CustomHLSL_14 (Custom%20HLSL :InputPins [(struct :Label In :AllowedTypes (struct :Ids [(struct :Struct /Script/PCG.PCGDataTypeInfoTexture2DSingleBase)]) :bAllowMultipleData false :PinStatus Required :bAllowMultipleConnections false)] :KernelType TextureProcessor :OutputPins [(struct :PropertiesGPU (struct :PinsToInititalizeFrom [In]) :Label Out :AllowedTypes (struct :Ids [(struct :Struct /Script/PCG.PCGDataTypeInfoTexture2DSingleBase)]) :bAllowMultipleData false)]))
  ;;[0.4375 0 0.2969 0.5] Debug Channel Textures (world)
  (bind CustomHLSL_0 (Custom%20HLSL :InputPins [(struct :Label In :AllowedTypes (struct :Ids [(struct :Struct /Script/PCG.PCGDataTypeInfoTexture2DArray)]) :bAllowMultipleData false :PinStatus Required :bAllowMultipleConnections false) (struct :Label SliceIndex :AllowedTypes (struct :Ids [(struct :Struct /Script/PCG.PCGDataTypeInfoParam)]) :bAllowMultipleData false :bAllowMultipleConnections false)] :KernelType Custom :OutputPins [(struct :PropertiesGPU (struct :PinsToInititalizeFrom [In]) :Label Out :AllowedTypes (struct :Ids [(struct :Struct /Script/PCG.PCGDataTypeInfoTexture2DSingleBase)]) :bAllowMultipleData false)]))
  (bind CustomHLSL_15 (Custom%20HLSL :AdditionalSources [/Hera/Environment/Terrain/MeshTerrain/Assets/ComputeSources/PCGCS_MeshTerrainHelpers.PCGCS_MeshTerrainHelpers] :bMuteUnwrittenPinDataErrors true :InputPins [(struct :Label In :AllowedTypes (struct :Ids [(struct :Struct /Script/PCG.PCGDataTypeInfoTexture2DSingleBase)]) :bAllowMultipleData false :PinStatus Required :bAllowMultipleConnections false) (struct :Label LocalPositions :AllowedTypes (struct :Ids [(struct :Struct /Script/PCG.PCGDataTypeInfoTexture2DSingleBase)]) :bAllowMultipleData false :bAllowMultipleConnections false) (struct :Label MeshTerrainTransform :AllowedTypes (struct :Ids [(struct :Struct /Script/PCG.PCGDataTypeInfoParam)]) :bAllowMultipleData false :bAllowMultipleConnections false)] :KernelType Custom :OutputPins [(struct :PropertiesGPU (struct :PinsToInititalizeFrom [In]) :Label Out :AllowedTypes (struct :Ids [(struct :Struct /Script/PCG.PCGDataTypeInfoPoint)]) :bAllowMultipleData false)]))
  (bind CustomHLSL_16 (Custom%20HLSL :OutputPins [(struct :PropertiesGPU (struct :PinsToInititalizeFrom [In]) :Label Out :AllowedTypes (struct :Ids [(struct :Struct /Script/PCG.PCGDataTypeInfoPoint)]))]))
  (bind StaticMeshSpawner_0 (Static%20Mesh%20Spawner :InstanceDataPackerParameters (class:/Script/PCG.PCGInstanceDataPackerByAttribute :AttributeSelectors [MyDensity MyExtents MyColor]) :InstanceDataPackerType /Script/PCG.PCGInstanceDataPackerByAttribute :MeshSelectorParameters (class:/Script/PCG.PCGMeshSelectorWeighted :MeshEntries [(struct :Descriptor (struct :StaticMesh /PCG/DebugObjects/PCG_Cube.PCG_Cube :OverrideMaterials [/PCG/DebugObjects/PCG_DebugMaterial.PCG_DebugMaterial]))])))
  ;;[0.4375 0 0.2969 0.5] Debug Mask (world)
  (bind CustomHLSL_17 (Custom%20HLSL :AdditionalSources [/Hera/Environment/Terrain/MeshTerrain/Assets/ComputeSources/PCGCS_MeshTerrainHelpers.PCGCS_MeshTerrainHelpers] :bMuteUnwrittenPinDataErrors true :InputPins [(struct :Label In :AllowedTypes (struct :Ids [(struct :Struct /Script/PCG.PCGDataTypeInfoTexture2DArray)]) :bAllowMultipleData false :PinStatus Required :bAllowMultipleConnections false) (struct :Label LocalPositions :AllowedTypes (struct :Ids [(struct :Struct /Script/PCG.PCGDataTypeInfoTexture2DSingleBase)]) :bAllowMultipleData false :bAllowMultipleConnections false) (struct :Label MeshTerrainTransform :AllowedTypes (struct :Ids [(struct :Struct /Script/PCG.PCGDataTypeInfoParam)]) :bAllowMultipleData false :bAllowMultipleConnections false) (struct :Label SliceIndex :AllowedTypes (struct :Ids [(struct :Struct /Script/PCG.PCGDataTypeInfoParam)]) :bAllowMultipleData false :bAllowMultipleConnections false) (struct :Label GrassMaps :AllowedTypes (struct :Ids [(struct :Struct /Script/PCG.PCGDataTypeInfoTexture2DArray)]) :bAllowMultipleData false :bAllowMultipleConnections false)] :KernelType Custom :OutputPins [(struct :PropertiesGPU (struct :PinsToInititalizeFrom [GrassMaps]) :Label Out :AllowedTypes (struct :Ids [(struct :Struct /Script/PCG.PCGDataTypeInfoPoint)]) :bAllowMultipleData false)]))
  (bind CustomHLSL_10 (Custom%20HLSL :OutputPins [(struct :PropertiesGPU (struct :PinsToInititalizeFrom [In]) :Label Out :AllowedTypes (struct :Ids [(struct :Struct /Script/PCG.PCGDataTypeInfoPoint)]))]))
  (bind StaticMeshSpawner_5 (Static%20Mesh%20Spawner :InstanceDataPackerParameters (class:/Script/PCG.PCGInstanceDataPackerByAttribute :AttributeSelectors [MyDensity MyExtents MyColor]) :InstanceDataPackerType /Script/PCG.PCGInstanceDataPackerByAttribute :MeshSelectorParameters (class:/Script/PCG.PCGMeshSelectorWeighted :MeshEntries [(struct :Descriptor (struct :StaticMesh /PCG/DebugObjects/PCG_Cube.PCG_Cube :OverrideMaterials [/PCG/DebugObjects/PCG_DebugMaterial.PCG_DebugMaterial]))])))
  (reroute ActualToKnownLGTs InActualToKnownLGTs)
  (reroute BakedLocalPositions BakeMeshTerrainSectionMesh_1)
  (reroute BakedMask BakeMeshTerrainSectionMesh_2)
  (reroute ChannelTextures CustomHLSL_22)
  (reroute CoverageMask CustomHLSL_21)
  (reroute DbgGrassIndex CreateConstant_3)
  (reroute GrassMaps Subgraph_1)
  (reroute GrassTypes InGrassTypes)
  (reroute InstancingGroups CustomHLSL_5.OutSeeds)
  (reroute LayerData InLayerData)
  (reroute LayerInfo CustomHLSL_4.LayerInfo)
  (reroute LocalPositions CustomHLSL_20)
  (reroute Mask DownsampleTexture_0)
  (reroute MeshSection In)
  (reroute MeshTerrainTransform CustomHLSL_1)
  (reroute NumScatteringThreads CustomHLSL_5.NumThreads)
  (reroute PrimData InPrimData)
  (reroute SourceChannelTextures GetMeshTerrainSectionChannelTextures_1)
  (reroute SourceChannelTexturesInfo GetTextureInfo_0)
  (reroute TexelSize InTexelSIze)
  (reroute TextureHeight ExtractAttributeAtIndex_1)
  (reroute TextureWidth ExtractAttributeAtIndex_0)
  (connect MeshSection Out BakeMeshTerrainSectionMesh_1 In)
  (connect ExtractAttributeAtIndex_4 Out BakeMeshTerrainSectionMesh_1 X)
  (connect ExtractAttributeAtIndex_5 Out BakeMeshTerrainSectionMesh_1 Y)
  (connect MeshSection Out BakeMeshTerrainSectionMesh_2 In)
  (connect ExtractAttributeAtIndex_4 Out BakeMeshTerrainSectionMesh_2 X)
  (connect ExtractAttributeAtIndex_5 Out BakeMeshTerrainSectionMesh_2 Y)
  (connect GetActorData_0 Out CopyAttributes_1 Source)
  (connect CreateConstant_1 Out CopyAttributes_1 Target)
  (connect DbgGrassIndex Out CustomHLSL_0 SliceIndex)
  (connect CopyAttributes_1 Out CustomHLSL_1 In)
  (connect GrassMaps Out CustomHLSL_2 In)
  (connect GrassMaps Out CustomHLSL_3 GrassMaps)
  (connect GrassMaps Out CustomHLSL_4 GrassMaps)
  (connect LayerData Out CustomHLSL_4 LayerData)
  (connect Mask Out CustomHLSL_4 Mask)
  (connect CustomHLSL_3 NumSlices CustomHLSL_4 NumElements)
  (connect TexelSize Out CustomHLSL_4 TexelSize)
  (connect GrassMaps Out CustomHLSL_5 GrassMaps)
  (connect LayerInfo Out CustomHLSL_5 LayerInfo)
  (connect Mask Out CustomHLSL_5 Mask)
  (connect CustomHLSL_4 TotalMaxNumCells CustomHLSL_5 NumElements)
  (connect TexelSize Out CustomHLSL_5 TexelSize)
  (connect GrassMaps Out CustomHLSL_6 GrassMap)
  (connect LayerData Out CustomHLSL_6 LayerData)
  (connect LayerInfo Out CustomHLSL_6 LayerInfo)
  (connect LocalPositions Out CustomHLSL_6 LocalPositions)
  (connect CoverageMask Out CustomHLSL_6 Mask)
  (connect MeshTerrainTransform Out CustomHLSL_6 MeshTerrainTransform)
  (connect NumScatteringThreads Out CustomHLSL_6 NumElements)
  (connect InstancingGroups Out CustomHLSL_6 Seeds)
  (connect TexelSize Out CustomHLSL_6 TexelSize)
  (connect DbgGrassIndex Out CustomHLSL_7 Layer)
  (connect LocalPositions Out CustomHLSL_7 LocalPositions)
  (connect MeshTerrainTransform Out CustomHLSL_7 MeshTerrainTransform)
  (connect CustomHLSL_7 Out CustomHLSL_8 In)
  (connect CustomHLSL_17 Out CustomHLSL_10 In)
  (connect CustomHLSL_25 Out CustomHLSL_11 In)
  (connect MeshTerrainTransform Out CustomHLSL_12 MeshTerrainTransform)
  (connect CustomHLSL_12 Out CustomHLSL_13 In)
  (connect CustomHLSL_0 Out CustomHLSL_15 In)
  (connect LocalPositions Out CustomHLSL_15 LocalPositions)
  (connect MeshTerrainTransform Out CustomHLSL_15 MeshTerrainTransform)
  (connect CustomHLSL_15 Out CustomHLSL_16 In)
  (connect GrassMaps Out CustomHLSL_17 GrassMaps)
  (connect LocalPositions Out CustomHLSL_17 LocalPositions)
  (connect MeshTerrainTransform Out CustomHLSL_17 MeshTerrainTransform)
  (connect DbgGrassIndex Out CustomHLSL_17 SliceIndex)
  (connect BakedLocalPositions Out CustomHLSL_19 LocalPositions)
  (connect MeshTerrainTransform Out CustomHLSL_19 MeshTerrainTransform)
  (connect TexelSize Out CustomHLSL_19 TexelSize)
  (connect BakedLocalPositions Out CustomHLSL_20 LocalPositions)
  (connect CustomHLSL_26 MinX CustomHLSL_20 MinX)
  (connect CustomHLSL_26 MinY CustomHLSL_20 MinY)
  (connect CustomHLSL_26 ResX CustomHLSL_20 NumElementsX)
  (connect CustomHLSL_26 ResY CustomHLSL_20 NumElementsY)
  (connect BakedMask Out CustomHLSL_21 LocalPositions)
  (connect CustomHLSL_26 MinX CustomHLSL_21 MinX)
  (connect CustomHLSL_26 MinY CustomHLSL_21 MinY)
  (connect CustomHLSL_26 ResX CustomHLSL_21 NumElementsX)
  (connect CustomHLSL_26 ResY CustomHLSL_21 NumElementsY)
  (connect SourceChannelTextures Out CustomHLSL_22 In)
  (connect CustomHLSL_26 MinX CustomHLSL_22 MinX)
  (connect CustomHLSL_26 MinY CustomHLSL_22 MinY)
  (connect CustomHLSL_26 ResX CustomHLSL_22 NumElementsX)
  (connect CustomHLSL_26 ResY CustomHLSL_22 NumElementsY)
  (connect ExtractAttributeAtIndex_3 Out CustomHLSL_22 NumElementsZ)
  (connect LocalPositions Out CustomHLSL_25 LocalPositions)
  (connect MeshTerrainTransform Out CustomHLSL_25 MeshTerrainTransform)
  (connect DbgGrassIndex Out CustomHLSL_25 SliceIndex)
  (connect CustomHLSL_19 Bounds CustomHLSL_26 In)
  (connect CustomHLSL_2 Out DownsampleTexture_0 In)
  (connect GetTextureInfo_1 Out ExtractAttributeAtIndex_0 In)
  (connect GetTextureInfo_1 Out ExtractAttributeAtIndex_1 In)
  (connect SourceChannelTexturesInfo Out ExtractAttributeAtIndex_3 In)
  (connect SourceChannelTexturesInfo Out ExtractAttributeAtIndex_4 In)
  (connect SourceChannelTexturesInfo Out ExtractAttributeAtIndex_5 In)
  (connect GetMeshTerrainSectionActor_0 Out GetActorData_0 In)
  (connect MeshSection Out GetMeshTerrainSectionActor_0 In)
  (connect MeshSection Out GetMeshTerrainSectionChannelTextures_1 In)
  (connect GetMeshTerrainSectionChannelTextures_1 Out GetTextureInfo_0 Texture)
  (connect ChannelTextures Out GetTextureInfo_1 Texture)
  (connect CustomHLSL_16 Out StaticMeshSpawner_0 In)
  (connect CustomHLSL_6 Out StaticMeshSpawner_1 In)
  (connect PrimData Out StaticMeshSpawner_1 PrimitiveTable)
  (connect CustomHLSL_8 Out StaticMeshSpawner_2 In)
  (connect CustomHLSL_11 Out StaticMeshSpawner_3 In)
  (connect CustomHLSL_13 Out StaticMeshSpawner_4 In)
  (connect CustomHLSL_10 Out StaticMeshSpawner_5 In)
  (connect ActualToKnownLGTs Out Subgraph_1 ActualToKnownLGTMap)
  (connect ChannelTextures Out Subgraph_1 Channel%20Textures)
  (connect CoverageMask Out Subgraph_1 CoverageMask)
  (connect TextureHeight Out Subgraph_1 Height)
  (connect LocalPositions Out Subgraph_1 LocalPositions)
  (connect MeshTerrainTransform Out Subgraph_1 MeshTerrainTransform)
  (connect TextureWidth Out Subgraph_1 Width))
