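; PCG mesh terrain modifier that displaces mesh partition vertices along their normals using a world-aligned height texture, defaulting to a cliff map.
; Queried points sample the texture in XZ and YZ planar projections from position divided by scale, then subtract an offset.
; Custom HLSL takes the normal-carrying points and blend parameters to supply ZMask, scaling displacement, and XMask, blending the projections.
; Displaced positions are written back on the given layer and priority.
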(graph PCG_MeshTerrain_Modifier_WorldAligned
  (params (Displacement :type (Object /Script/Engine.Texture2D) :default /Hermes/Terrain/Textures/T_Hermes_Cliff_Default_A_H.T_Hermes_Cliff_Default_A_H) (Disp_Scale :type Double :default 4e+03) (Disp_Amt :type Double :default 384.0) (Blend_Width :type Double :default 0.3) (Blend_Threshold :type Double :default 0.8) (Offset :type Double :default 0.5) (Priority :type Double) (Blend_XY_Gradient :type Double :default 1.0) (LayerName :type Name))
  (bind MegaMeshQuery_0 (Mesh%20Partition%20Query :QueryParams (struct :QueryType Intermediate :bRecomputeVertexNormals true :Channels [Gravel Mud] :bGetImpactNormal true) :LayerName LayerName :SubPriority Priority))
  (bind ToPoint_0 (To%20Point))
  (bind CopyAttributes_0 (Copy%20Attributes :InputSource $Position :OutputTarget SourcePositions))
  (bind FilterDataByType_0 (Filter%20Data%20By%20Type :TargetType (struct :Ids [(struct :Struct /Script/PCG.PCGDataTypeInfoPoint)])))
  (bind AttributeMathsOp_3 (Attribute%20Maths%20Op :Operation Multiply :InA Disp_Scale))
  (bind AttributeMathsOp_0 (Attribute%20Maths%20Op :InputSource1 $Position :Operation Divide :OutputTarget @Source))
  (bind GetTextureData_1 (Get%20Texture%20Data :ColorChannel Red :Texture-param Displacement))
  (bind SampleTexture_0 (Sample%20Texture :TextureMappingMethod UVCoordinates :UVCoordinatesAttribute $Position.RB))
  (bind AttributeVectorOp_0 (Attribute%20Vector%20Op :InputSource2 $Transform :Operation TransformDirection :OutputTarget Normal))
  (bind AddAttribute_0 (Add%20Attribute :AttributeTypes (struct :Type Float) :OutputTarget ZMask))
  (bind AddAttribute_1 (Add%20Attribute :AttributeTypes (struct :Type Float) :OutputTarget XMask))
  (bind CustomHLSL_1 (Custom%20HLSL :InputPins [(struct :Label Points :AllowedTypes (struct :Ids [(struct :Struct /Script/PCG.PCGDataTypeInfoPoint)]) :bAllowMultipleData false :bAllowMultipleConnections false) (struct :Label BlendGradient :AllowedTypes (struct :Ids [(struct :Struct /Script/PCG.PCGDataTypeInfoParam)]) :bAllowMultipleData false :bAllowMultipleConnections false) (struct :Label BlendThreshold :AllowedTypes (struct :Ids [(struct :Struct /Script/PCG.PCGDataTypeInfoParam)]) :bAllowMultipleConnections false) (struct :Label XYGradient :AllowedTypes (struct :Ids [(struct :Struct /Script/PCG.PCGDataTypeInfoParam)]) :bAllowMultipleData false :bAllowMultipleConnections false)] :NumElements 1000 :OutputPins [(struct :PropertiesGPU (struct) :Label Outpoints :AllowedTypes (struct :Ids [(struct :Struct /Script/PCG.PCGDataTypeInfoPoint)]) :bAllowMultipleData false)] :BlendGradient Blend_Width :BlendThreshold Blend_Threshold :XYGradient Blend_XY_Gradient))
  (bind AttributeMathsOp_1 (Attribute%20Maths%20Op :InputSource1 @Last :InputSource2 ZMask :Operation Multiply :InA Disp_Amt))
  (bind SampleTexture_1 (Sample%20Texture :TextureMappingMethod UVCoordinates :UVCoordinatesAttribute $Position.GB))
  (bind AttributeMathsOp_6 (Attribute%20Maths%20Op :InputSource1 $Color :Operation Subtract :OutputTarget $Color :InB Offset))
  (bind AttributeMathsOp_2 (Attribute%20Maths%20Op :InputSource1 $Color :Operation Multiply))
  (bind AttributeMathsOp_4 (Attribute%20Maths%20Op :InputSource1 $Color :Operation Subtract :OutputTarget $Color :InB Offset))
  (bind AttributeMathsOp_10 (Attribute%20Maths%20Op :InputSource1 $Color :Operation Multiply))
  (bind AttributeMathsOp_5 (Attribute%20Maths%20Op :InputSource1 $Color :InputSource2 $Color :InputSource3 XMask :Operation Lerp))
  (bind AttributeMathsOp_12 (Attribute%20Maths%20Op :InputSource2 $Color.R :Operation Multiply :OutputTarget $Position))
  (bind AttributeMathsOp_9 (Attribute%20Maths%20Op :InputSource1 $Position :InputSource2 $Position))
  (bind MegaMeshWrite_1 (MeshPartition%20Write :Priority 1e+02 :Type None :Priority-param Priority :Type-param LayerName))
  (connect AttributeVectorOp_0 Out AddAttribute_0 In)
  (connect AddAttribute_0 Out AddAttribute_1 In)
  (connect FilterDataByType_0 InsideFilter AttributeMathsOp_0 InA)
  (connect AttributeMathsOp_3 Out AttributeMathsOp_0 InB)
  (connect CustomHLSL_1 Outpoints AttributeMathsOp_1 InB)
  (connect AttributeMathsOp_6 Out AttributeMathsOp_2 InA)
  (connect AttributeMathsOp_1 Out AttributeMathsOp_2 InB)
  (connect SampleTexture_0 Out AttributeMathsOp_4 InA)
  (connect AttributeMathsOp_10 Out AttributeMathsOp_5 InA)
  (connect AttributeMathsOp_2 Out AttributeMathsOp_5 InB)
  (connect CustomHLSL_1 Outpoints AttributeMathsOp_5 Ratio)
  (connect SampleTexture_1 Out AttributeMathsOp_6 InA)
  (connect FilterDataByType_0 InsideFilter AttributeMathsOp_9 InA)
  (connect AttributeMathsOp_12 Out AttributeMathsOp_9 InB)
  (connect AttributeMathsOp_4 Out AttributeMathsOp_10 InA)
  (connect AttributeMathsOp_1 Out AttributeMathsOp_10 InB)
  (connect AttributeVectorOp_0 Out AttributeMathsOp_12 InA)
  (connect AttributeMathsOp_5 Out AttributeMathsOp_12 InB)
  (connect FilterDataByType_0 InsideFilter AttributeVectorOp_0 Transform)
  (connect ToPoint_0 Out CopyAttributes_0 Source)
  (connect ToPoint_0 Out CopyAttributes_0 Target)
  (connect AddAttribute_1 Out CustomHLSL_1 Points)
  (connect CopyAttributes_0 Out FilterDataByType_0 In)
  (connect AttributeMathsOp_9 Out MegaMeshWrite_1 In)
  (connect GetTextureData_1 Out SampleTexture_0 BaseTexture)
  (connect AttributeMathsOp_0 Out SampleTexture_0 Point)
  (connect GetTextureData_1 Out SampleTexture_1 BaseTexture)
  (connect AttributeMathsOp_0 Out SampleTexture_1 Point)
  (connect MegaMeshQuery_0 Out ToPoint_0 In))
