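; PCG graph that post-processes MeshTerrain, deriving physical material weight channels from painted material and biome layers.
; It queries the mesh partition as points, thresholds normal slope into a Cliff mask and inverts combined layers into Grass.
; Dirt combines gravel, mud and Biome_B minus other masks, Stone copies Cliff, and only these two are written back.
; Grass, Asphalt, Sand and Snow PhysMat branches remain unwritten, alongside density-based debug previews.
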(graph PCG_MeshTerrain_PostProcess
  (bind CreateConstant_0 (Create%20Constant :AttributeTypes (struct :Type Name :NameValue Post%20Process) :OutputTarget PriorityLayer))
  (bind MegaMeshQuery_1 (Mesh%20Partition%20Query :QueryParams (struct :QueryType Final :LayerName Post%20Process :bRecomputeVertexNormals true :Channels [Mat_Gravel Mat_Mud Mat_Sand Mat_Asphalt Biome_A Biome_B Biome_C Biome_D])))
  (bind ToPoint_1 (To%20Point))
  (bind AddAttribute_0 (Add%20Attribute :AttributeTypes (struct :Type Float :VectorValue (0.0 0.0 1.0)) :OutputTarget Mask))
  (bind MegaMeshWrite_1 (MeshPartition%20Write :bWritePositions false :Channels [PhysMat_Dirt PhysMat_Stone] :Type None))
  ;[0.15 0.15 0.15 0.5] Cliff
  (bind CopyAttributes_1 (Copy%20Attributes :InputSource $Position :OutputTarget SourcePositions))
  (bind AddAttribute_3 (Add%20Attribute :AttributeTypes (struct :Type Vector :VectorValue (0.0 0.0 1.0)) :OutputTarget Normal))
  (bind AttributeVectorOp_0 (Attribute%20Vector%20Op :InputSource2 $Transform :Operation TransformDirection :OutputTarget Normal))
  (bind AttributeRemap_0 (Attribute%20Remap :bClampToUnitRange true :InputSource Normal.Z :InRangeMax 0.8 :InRangeMin 0.79 :OutputTarget Cliff :OutRangeMax 0.0 :OutRangeMin 1.0))
  (bind CopyAttributes_0 (Copy%20Attributes :InputSource @Last :OutputTarget $Density)) ; DEBUG
  ;[0.15 0.15 0.15 0.5] Grass: Generate mask by inversing existing layers after combining them
  (bind AddAttribute_4 (Add%20Attribute :OutputTarget Grass)) ; Mask
  (bind AttributeMathsOp_9 (Attribute%20Maths%20Op :InputSource1 Mat_Gravel :InputSource2 Mat_Mud :Operation Max :OutputTarget Grass)) ; Gravel + Mud
  (bind AttributeMathsOp_10 (Attribute%20Maths%20Op :InputSource1 Mat_Sand :InputSource2 Mat_Asphalt :Operation Max :OutputTarget Grass)) ; Sand + Asphalt
  (bind AttributeMathsOp_11 (Attribute%20Maths%20Op :InputSource1 Grass :InputSource2 Grass :Operation Max :OutputTarget Grass)) ; Inverse Grass
  (bind AttributeMathsOp_13 (Attribute%20Maths%20Op :InputSource1 Biome_C :InputSource2 Biome_D :Operation Max :OutputTarget Grass)) ; BiomeC BiomeD
  (bind AttributeMathsOp_14 (Attribute%20Maths%20Op :InputSource1 Grass :InputSource2 Grass :Operation Max :OutputTarget Grass)) ; Inverse Grass
  (bind AttributeMathsOp_15 (Attribute%20Maths%20Op :InputSource1 Grass :InputSource2 Grass :Operation Max :OutputTarget Grass)) ; Inverse Grass
  (bind AttributeMathsOp_16 (Attribute%20Maths%20Op :InputSource1 Grass :InputSource2 Cliff :Operation Max :OutputTarget Grass)) ; Inverse Grass
  (bind AttributeMathsOp_32 (Attribute%20Maths%20Op :Operation Multiply))
  (bind AttributeMathsOp_33 (Attribute%20Maths%20Op :Operation Clamp))
  (bind AttributeMathsOp_17 (Attribute%20Maths%20Op :InputSource1 Grass :Operation OneMinus :OutputTarget Grass))
  (bind CopyAttributes_3 (Copy%20Attributes :InputSource @Last :OutputTarget $Density)) ; Debug
  ;[0.15 0.15 0.15 0.5] Note: BiomeB removed from here since ForestFloor
  ; is considered Grass in the main Material
  (bind CreateConstant_1 (Create%20Constant :AttributeTypes (struct :Type Float)))
  (bind AttributeMathsOp_12 (Attribute%20Maths%20Op :InputSource2 Biome_B :Operation Max :OutputTarget Grass)) ; 0 + BiomeB
  ;[0 0.1256 0.5625 0.5] PhysMat_Dirt
  (bind AddAttribute_1 (Add%20Attribute :OutputTarget PhysMat_Dirt)) ; PhysMat_Dirt
  (bind AttributeMathsOp_1 (Attribute%20Maths%20Op :InputSource1 Mat_Gravel :InputSource2 Mat_Mud :Operation Max :OutputTarget PhysMat_Dirt)) ; Gravel Mud
  (bind AttributeMathsOp_2 (Attribute%20Maths%20Op :InputSource1 PhysMat_Dirt :InputSource2 Biome_B :Operation Max :OutputTarget PhysMat_Dirt)) ; Scree
  (bind AddAttribute_2 (Add%20Attribute :OutputTarget Mask)) ; Mask
  (bind AttributeMathsOp_6 (Attribute%20Maths%20Op :InputSource1 Cliff :InputSource2 Grass :Operation Max :OutputTarget Mask)) ; Cliff Grass
  (bind AttributeMathsOp_7 (Attribute%20Maths%20Op :InputSource1 Mask :InputSource2 Mat_Sand :Operation Max :OutputTarget Mask)) ; Sand
  (bind AttributeMathsOp_8 (Attribute%20Maths%20Op :InputSource1 Mask :InputSource2 Mat_Asphalt :Operation Max :OutputTarget Mask)) ; Asphalt
  (bind AttributeMathsOp_5 (Attribute%20Maths%20Op :Operation OneMinus))
  (bind AttributeMathsOp_3 (Attribute%20Maths%20Op :InputSource2 Mask :Operation Multiply :OutputTarget PhysMat_Dirt))
  (bind CopyAttributes_2 (Copy%20Attributes :InputSource @Last :OutputTarget $Density)) ; DEBUG
  ;[0 0.1256 0.5625 0.5] PhysMat_Stone
  (bind AddAttribute_5 (Add%20Attribute :OutputTarget PhysMat_Stone)) ; PhysMat_Stone
  (bind AttributeMathsOp_18 (Attribute%20Maths%20Op :InputSource1 Cliff :InputSource2 Cliff :Operation Set :OutputTarget PhysMat_Stone))
  (bind CopyAttributes_4 (Copy%20Attributes :InputSource @Last :OutputTarget $Density)) ; DEBUG
  ;[0 0.1256 0.5625 0.5] PhysMat_Grass
  ; Change in 41.2 to remove PhysMat for grass as it is already the Default PhysMat in the MPD. Optimization.
  (bind AddAttribute_6 (Add%20Attribute :OutputTarget PhysMat_Grass)) ; PhysMat_Grass
  (bind AttributeMathsOp_19 (Attribute%20Maths%20Op :InputSource1 Grass :InputSource2 Grass :Operation Set :OutputTarget PhysMat_Grass))
  (bind CopyAttributes_5 (Copy%20Attributes :InputSource @Last :OutputTarget $Density)) ; DEBUG
  ;[0 0.1256 0.5625 0.5] PhysMat_Asphalt
  ; Change in 41.2 to use Mat_Asphalt as PhysMat so no need to copy here
  (bind AddAttribute_7 (Add%20Attribute :OutputTarget PhysMat_Asphalt)) ; PhysMat_Asphalt
  (bind AttributeMathsOp_20 (Attribute%20Maths%20Op :InputSource1 Mat_Asphalt :InputSource2 Mat_Asphalt :Operation Set :OutputTarget PhysMat_Asphalt))
  (bind CopyAttributes_6 (Copy%20Attributes :InputSource @Last :OutputTarget $Density)) ; DEBUG
  ;[0 0.1256 0.5625 0.5] PhysMat_Sand
  ; Change in 41.2 to use Mat_Sand as PhysMat so no need to copy here
  (bind AddAttribute_8 (Add%20Attribute :OutputTarget PhysMat_Sand)) ; PhysMat_Sand
  (bind AttributeMathsOp_21 (Attribute%20Maths%20Op :InputSource1 Mat_Sand :InputSource2 Mat_Sand :Operation Set :OutputTarget PhysMat_Sand))
  (bind AttributeMathsOp_22 (Attribute%20Maths%20Op :InputSource1 Cliff :Operation OneMinus :OutputTarget Cliff))
  (bind AttributeMathsOp_23 (Attribute%20Maths%20Op :InputSource1 PhysMat_Sand :InputSource2 Cliff :Operation Multiply :OutputTarget PhysMat_Sand))
  (bind CopyAttributes_7 (Copy%20Attributes :InputSource @Last :OutputTarget $Density)) ; DEBUG
  ;[0 0.1256 0.5625 0.5] PhysMat_Snow
  (bind AddAttribute_10 (Add%20Attribute :OutputTarget PhysMat_Snow)) ; PhysMat_Sand
  (bind AttributeMathsOp_24 (Attribute%20Maths%20Op :InputSource1 Biome_C :InputSource2 Biome_C :Operation Set :OutputTarget PhysMat_Snow))
  (bind AttributeMathsOp_25 (Attribute%20Maths%20Op :InputSource1 Cliff :Operation OneMinus :OutputTarget Cliff)) ; Cliff
  (bind AttributeMathsOp_26 (Attribute%20Maths%20Op :InputSource1 PhysMat_Snow :InputSource2 Cliff :Operation Multiply :OutputTarget PhysMat_Snow))
  (bind CopyAttributes_8 (Copy%20Attributes :InputSource @Last :OutputTarget $Density)) ; DEBUG
  ;[0.15 0.15 0.15 0.5] Debug Combined Masks: Black Spots Here indicate nothing specified for Physical Material
  (bind AddAttribute_9 (Add%20Attribute :OutputTarget DEBUGMASK)) ; PhysMat_Dirt
  (bind AttributeMathsOp_0 (Attribute%20Maths%20Op :InputSource1 PhysMat_Dirt :InputSource2 PhysMat_Stone :Operation Max :OutputTarget DEBUGMASK)) ; DIRT + STONE
  (bind AttributeMathsOp_4 (Attribute%20Maths%20Op :InputSource2 PhysMat_Grass :Operation Max :OutputTarget DEBUGMASK)) ; GRASS
  (bind AttributeMathsOp_27 (Attribute%20Maths%20Op :InputSource2 Mat_Asphalt :Operation Max :OutputTarget DEBUGMASK)) ; Asphalt
  (bind AttributeMathsOp_28 (Attribute%20Maths%20Op :InputSource2 Mat_Sand :Operation Max :OutputTarget DEBUGMASK)) ; Sand
  (bind CopyAttributes_9 (Copy%20Attributes :InputSource @Last :OutputTarget $Density)) ; DEBUG
  (connect ToPoint_1 Out AddAttribute_0 In)
  (connect AttributeMathsOp_17 Out AddAttribute_1 In)
  (connect AttributeMathsOp_17 Out AddAttribute_2 In)
  (connect CopyAttributes_1 Out AddAttribute_3 In)
  (connect AttributeRemap_0 Out AddAttribute_4 In)
  (connect AttributeMathsOp_3 Out AddAttribute_5 In)
  (connect AttributeMathsOp_18 Out AddAttribute_6 In)
  (connect AttributeMathsOp_19 Out AddAttribute_7 In)
  (connect AttributeMathsOp_20 Out AddAttribute_8 In)
  (connect AttributeMathsOp_18 Out AddAttribute_9 In)
  (connect AddAttribute_9 Out AttributeMathsOp_0 InA)
  (connect AddAttribute_9 Out AttributeMathsOp_0 InB)
  (connect AddAttribute_1 Out AttributeMathsOp_1 InA)
  (connect AddAttribute_1 Out AttributeMathsOp_1 InB)
  (connect AttributeMathsOp_1 Out AttributeMathsOp_2 InA)
  (connect AttributeMathsOp_1 Out AttributeMathsOp_2 InB)
  (connect AttributeMathsOp_2 Out AttributeMathsOp_3 InA)
  (connect AttributeMathsOp_5 Out AttributeMathsOp_3 InB)
  (connect AttributeMathsOp_8 Out AttributeMathsOp_5 In)
  (connect AddAttribute_2 Out AttributeMathsOp_6 InA)
  (connect AddAttribute_2 Out AttributeMathsOp_6 InB)
  (connect AttributeMathsOp_6 Out AttributeMathsOp_7 InA)
  (connect AttributeMathsOp_6 Out AttributeMathsOp_7 InB)
  (connect AttributeMathsOp_7 Out AttributeMathsOp_8 InA)
  (connect AttributeMathsOp_7 Out AttributeMathsOp_8 InB)
  (connect AddAttribute_4 Out AttributeMathsOp_9 InA)
  (connect AddAttribute_4 Out AttributeMathsOp_9 InB)
  (connect AddAttribute_4 Out AttributeMathsOp_10 InA)
  (connect AddAttribute_4 Out AttributeMathsOp_10 InB)
  (connect AttributeMathsOp_9 Out AttributeMathsOp_11 InA)
  (connect AttributeMathsOp_10 Out AttributeMathsOp_11 InB)
  (connect CreateConstant_1 Out AttributeMathsOp_12 InA)
  (connect AddAttribute_4 Out AttributeMathsOp_12 InB)
  (connect AddAttribute_4 Out AttributeMathsOp_13 InA)
  (connect AddAttribute_4 Out AttributeMathsOp_13 InB)
  (connect AttributeMathsOp_12 Out AttributeMathsOp_14 InA)
  (connect AttributeMathsOp_13 Out AttributeMathsOp_14 InB)
  (connect AttributeMathsOp_11 Out AttributeMathsOp_15 InA)
  (connect AttributeMathsOp_14 Out AttributeMathsOp_15 InB)
  (connect AttributeMathsOp_15 Out AttributeMathsOp_16 InA)
  (connect AttributeMathsOp_15 Out AttributeMathsOp_16 InB)
  (connect AttributeMathsOp_33 Out AttributeMathsOp_17 In)
  (connect AddAttribute_5 Out AttributeMathsOp_18 InA)
  (connect AddAttribute_5 Out AttributeMathsOp_18 InB)
  (connect AddAttribute_6 Out AttributeMathsOp_19 InA)
  (connect AddAttribute_6 Out AttributeMathsOp_19 InB)
  (connect AddAttribute_7 Out AttributeMathsOp_20 InA)
  (connect AddAttribute_7 Out AttributeMathsOp_20 InB)
  (connect AddAttribute_8 Out AttributeMathsOp_21 InA)
  (connect AddAttribute_8 Out AttributeMathsOp_21 InB)
  (connect AddAttribute_8 Out AttributeMathsOp_22 In)
  (connect AttributeMathsOp_21 Out AttributeMathsOp_23 InA)
  (connect AttributeMathsOp_22 Out AttributeMathsOp_23 InB)
  (connect AddAttribute_10 Out AttributeMathsOp_24 InA)
  (connect AddAttribute_10 Out AttributeMathsOp_24 InB)
  (connect AddAttribute_10 Out AttributeMathsOp_25 In)
  (connect AttributeMathsOp_24 Out AttributeMathsOp_26 InA)
  (connect AttributeMathsOp_25 Out AttributeMathsOp_26 InB)
  (connect AttributeMathsOp_0 Out AttributeMathsOp_27 InA)
  (connect AttributeMathsOp_0 Out AttributeMathsOp_27 InB)
  (connect AttributeMathsOp_27 Out AttributeMathsOp_28 InA)
  (connect AttributeMathsOp_27 Out AttributeMathsOp_28 InB)
  (connect AttributeMathsOp_16 Out AttributeMathsOp_32 InA)
  (connect AttributeMathsOp_32 Out AttributeMathsOp_33 In)
  (connect AttributeVectorOp_0 Out AttributeRemap_0 In)
  (connect AddAttribute_3 Out AttributeVectorOp_0 In)
  (connect AddAttribute_3 Out AttributeVectorOp_0 Transform)
  (connect AttributeRemap_0 Out CopyAttributes_0 Source)
  (connect AttributeRemap_0 Out CopyAttributes_0 Target)
  (connect AddAttribute_0 Out CopyAttributes_1 Source)
  (connect AddAttribute_0 Out CopyAttributes_1 Target)
  (connect AttributeMathsOp_3 Out CopyAttributes_2 Source)
  (connect AttributeMathsOp_3 Out CopyAttributes_2 Target)
  (connect AttributeMathsOp_17 Out CopyAttributes_3 Source)
  (connect AttributeMathsOp_17 Out CopyAttributes_3 Target)
  (connect AttributeMathsOp_18 Out CopyAttributes_4 Source)
  (connect AttributeMathsOp_18 Out CopyAttributes_4 Target)
  (connect AttributeMathsOp_19 Out CopyAttributes_5 Source)
  (connect AttributeMathsOp_19 Out CopyAttributes_5 Target)
  (connect AttributeMathsOp_20 Out CopyAttributes_6 Source)
  (connect AttributeMathsOp_20 Out CopyAttributes_6 Target)
  (connect AttributeMathsOp_23 Out CopyAttributes_7 Source)
  (connect AttributeMathsOp_23 Out CopyAttributes_7 Target)
  (connect AttributeMathsOp_26 Out CopyAttributes_8 Source)
  (connect AttributeMathsOp_26 Out CopyAttributes_8 Target)
  (connect AttributeMathsOp_28 Out CopyAttributes_9 Source)
  (connect AttributeMathsOp_28 Out CopyAttributes_9 Target)
  (connect CreateConstant_0 Out MegaMeshQuery_1 LayerName)
  (connect AttributeMathsOp_18 Out MegaMeshWrite_1 In)
  (connect CreateConstant_0 Out MegaMeshWrite_1 Type)
  (connect MegaMeshQuery_1 Out ToPoint_1 In))
