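; PCG graph for the dirt road that spawns dirt road spline meshes along a spline, with a mesh-terrain material and LobbyIsland runtime virtual texture output.
; Spline Z position is adjusted by Y scale before reaching a Select node that feeds the mesh spawner.
; A work-in-progress branch attracts control points to nearby dirt road actors' "Road"-tagged splines, rebuilds the spline for the Select, and writes it back through a Blueprint element.
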(graph PCG_MeshTerrain_Road_Dirt
  (bind GetSplineData_1 (Get%20Spline%20Data))
  (bind AttributeMathsOp_1 (Attribute%20Maths%20Op :InputSource1 $Scale.Y :InputSource2 $Scale.Y :Operation Set :OutputTarget $Scale.Y))
  (bind AttributeMathsOp_3 (Attribute%20Maths%20Op :InputSource1 $Position.Z :InputSource2 $Scale.Y :OutputTarget $Position.Z))
  (bind Select_0 (Select))
  (bind SpawnSplineMesh_0 (Spawn%20Spline%20Mesh :SplineMeshDescriptor (struct :StaticMesh /Game/Environments/Asteria/Terrain/DirtRoad/SM_DirtRoad.SM_DirtRoad :OverrideMaterials [/Hera/Environment/Terrain/MeshTerrain/Materials/MI_DirtPath_MeshTerrain.MI_DirtPath_MeshTerrain] :RuntimeVirtualTextures [/Hera/Environment/Terrain/MeshTerrain/VirtualTextures/RVT_Hera_MeshPartition_LobbyIsland.RVT_Hera_MeshPartition_LobbyIsland] :BodyInstance (struct :CollisionEnabled QueryOnly :CollisionProfileName BlockAll :CollisionResponses (struct :ResponseArray [(struct :Channel FCC_Trace_WeaponCapsule) (struct :Channel FCC_Object_PawnAI) (struct :Channel FCC_Trace_WeaponMulti) (struct :Channel FCC_Trace_Weapon) (struct :Channel FCC_Object_Projectile) (struct :Channel FCC_Object_ProjectileIgnoreAI) (struct :Channel FCC_Object_ProjectileMesh) (struct :Channel FCC_Object_SpectatorDroneCamera) (struct :Channel FCC_Trace_WeaponPawnAndWorldMulti)]) :bAutoWeld true) :TranslucencySortPriority 1 :bCastShadow false)))
  ;[0.15 0.15 0.15 0.5] WIP auto snap to other dirt road points
  (bind GetSplineControlPoints_0 (Get%20Spline%20Control%20Points))
  (bind GetActorData_0 (Get%20Actor%20Data :ActorSelector (struct :ActorFilter AllWorldActors :ActorSelection ByClass :ActorSelectionClass /Hera/Environment/Terrain/MeshTerrain/Assets/BP_MeshTerrain_Road_Dirt.BP_MeshTerrain_Road_Dirt_C :bSelectMultiple true :bIgnoreSelfAndChildren true) :ComponentSelector (struct :ComponentSelectionTag Road)))
  (bind FilterDataByType_0 (Filter%20Data%20By%20Type :TargetType (struct :Ids [(struct :Struct /Script/PCG.PCGDataTypeInfoPolyline)])))
  (bind GetSplineControlPoints_1 (Get%20Spline%20Control%20Points))
  (bind AttractElement_1 (Attract :bOutputAttractIndex true :Distance 5e+02 :SourceAndTargetAttributeMapping (map ($Position $Position) ($Rotation $Rotation))))
  (bind CreateSpline_1 (Create%20Spline :bApplyCustomTangents true :InterpTypeAttribute None))
  ;;[0.15 0.15 0.15 0.5] Update the original spline
  (bind ExecuteBlueprint_0 (Execute%20Blueprint :BlueprintElementInstance (class:/Hera/Environment/Terrain/MeshTerrain/Assets/BP_PCG_UpdateSplineComponent.BP_PCG_UpdateSplineComponent_C) :BlueprintElementType /Hera/Environment/Terrain/MeshTerrain/Assets/BP_PCG_UpdateSplineComponent.BP_PCG_UpdateSplineComponent_C))
  (connect GetSplineControlPoints_0 Out AttractElement_1 Source)
  (connect GetSplineControlPoints_1 Out AttractElement_1 Target)
  (connect GetSplineData_1 Out AttributeMathsOp_1 InA)
  (connect AttributeMathsOp_1 Out AttributeMathsOp_3 InA)
  (connect AttractElement_1 Out CreateSpline_1 In)
  (connect CreateSpline_1 Out ExecuteBlueprint_0 Spline)
  (connect GetActorData_0 Out FilterDataByType_0 In)
  (connect AttributeMathsOp_3 Out GetSplineControlPoints_0 In)
  (connect FilterDataByType_0 InsideFilter GetSplineControlPoints_1 In)
  (connect AttributeMathsOp_3 Out Select_0 Input%20A)
  (connect CreateSpline_1 Out Select_0 Input%20B)
  (connect Select_0 Out SpawnSplineMesh_0 In))
