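; PCG graph that reshapes mesh terrain around a closed or open spline with a noise-broken edge, writing back by layer and priority.
; Sampled spline points are displaced sideways by either two-axis texture sampling or procedural spatial noise.
; Queried mesh points are projected onto the dilated spline polygon, then blended using curve-remapped edge distance combined with an interior mask.
; A further distance-weighted offset along SplineVector is computed but never reaches the write.
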(graph PCG_MeshTerrain_SplineWithNoise_01
  (params (Displacement_Amount :type Double :default 512.0) (Edge_Width :type Double :default 512.0) (Procedural_Noise_Tiling :type Double :default 16.0) (Layer :type Name) (Priority :type Int) (Offset :type Double :default 0.5) (Disp_Scale :type Double :default -4e+03) (Displacement :type (Object /Script/Engine.Texture2D) :default /Engine/Functions/Engine_MaterialFunctions02/PivotPainter2/Black_1x1_EXR_Texture.Black_1x1_EXR_Texture) (Use_World_Aligned_Texture :type Bool :default true))
  (bind MegaMeshQuery_1 (Mesh%20Partition%20Query :QueryParams (struct :QueryType Intermediate) :LayerName Layer :SubPriority Priority)) ; Here select at which priority \nto read the mesh partition from (Query Type)\nAdditionaly you can load specific channels as attributes.
  (bind ToPoint_1 (To%20Point)) ; Convert surface to points
  (bind CopyAttributes_1 (Copy%20Attributes :InputSource $Position :OutputTarget SourcePositions)) ; Copy Source Positions
  (bind AddAttribute_11 (Add%20Attribute :AttributeTypes (struct :Type Vector) :OutputTarget SplineVector))
  (bind GetSplineData_0 (Get%20Spline%20Data))
  (bind FilterDataByType_1 (Filter%20Data%20By%20Type :TargetType (struct :Ids [(struct :Struct /Script/PCG.PCGDataTypeInfoSpline)])))
  (bind SplineDirection_0 (Spline%20Direction :Operation ForceClockwise))
  (bind SplineSampler_1 (Spline%20Sampler :SamplerParams (struct :Mode Distance :SubdivisionsPerSegment 4 :DistanceIncrement 32.0 :InteriorSampleSpacing 16.0 :InteriorBorderSampleSpacing 16.0 :PointSteepness 1.0)))
  (bind AddAttribute_7 (Add%20Attribute :AttributeTypes (struct :Type Vector) :OutputTarget SplineNormal))
  (bind CopyAttributes_3 (Copy%20Attributes :InputSource $Rotation.Right :OutputTarget SplineNormal))
  (bind AddAttribute_8 (Add%20Attribute :AttributeTypes (struct :Type Float) :OutputTarget XMask))
  (bind GetActorProperty_0 (Get%20Actor%20Property :PropertyName Displacement%20Amount))
  (bind AttributeMathsOp_1 (Attribute%20Maths%20Op :Operation Multiply))
  (bind GetActorProperty_3 (Get%20Actor%20Property :PropertyName Offset))
  (bind GetActorProperty_4 (Get%20Actor%20Property :PropertyName Displacement%20Scale))
  (bind AttributeMathsOp_3 (Attribute%20Maths%20Op :Operation Multiply))
  (bind AttributeMathsOp_14 (Attribute%20Maths%20Op :InputSource1 $Position :Operation Divide))
  (bind GetTextureData_2 (Get%20Texture%20Data :Texture-param Displacement))
  (bind SampleTexture_1 (Sample%20Texture :TextureMappingMethod UVCoordinates :UVCoordinatesAttribute $Position.RB))
  (bind AttributeMathsOp_19 (Attribute%20Maths%20Op :InputSource1 $Color :Operation Subtract :OutputTarget $Color))
  (bind AttributeMathsOp_30 (Attribute%20Maths%20Op :Operation Multiply))
  (bind SampleTexture_2 (Sample%20Texture :TextureMappingMethod UVCoordinates :UVCoordinatesAttribute $Position.GB))
  (bind AttributeMathsOp_24 (Attribute%20Maths%20Op :InputSource1 $Color :Operation Subtract :OutputTarget $Color))
  (bind AttributeMathsOp_31 (Attribute%20Maths%20Op :Operation Multiply))
  (bind CustomHLSL_1 (Custom%20HLSL :InputPins [(struct :Label Points :AllowedTypes (struct :Ids [(struct :Struct /Script/PCG.PCGDataTypeInfoPoint)]) :bAllowMultipleData false :bAllowMultipleConnections false)] :NumElements 1000 :OutputPins [(struct :PropertiesGPU (struct) :Label Outpoints :AllowedTypes (struct :Ids [(struct :Struct /Script/PCG.PCGDataTypeInfoPoint)]) :bAllowMultipleData false)]))
  (bind AttributeMathsOp_21 (Attribute%20Maths%20Op :InputSource1 $Color :InputSource2 $Color :InputSource3 XMask :Operation Lerp))
  (bind AttributeMathsOp_17 (Attribute%20Maths%20Op :InputSource1 $Color.R :InputSource2 SplineNormal :Operation Multiply :OutputTarget $Position))
  (bind AttributeMathsOp_15 (Attribute%20Maths%20Op :InputSource1 $Position :InputSource2 $Position))
  (bind GetActorProperty_7 (Get%20Actor%20Property :PropertyName Use%20World%20Aligned%20Texture))
  (bind Select_0 (Select))
  (bind GetActorData_0 (Get%20Actor%20Data :bMergeSinglePointData true :Mode GetSinglePoint))
  (bind AttributeMathsOp_5 (Attribute%20Maths%20Op :InputSource1 $Position.Z :InputSource2 $Position.Z :Operation Set :OutputTarget $Position.Z))
  (bind GetActorData_1 (Get%20Actor%20Data :bMergeSinglePointData true :Mode GetSinglePoint))
  (bind ToPoint_2 (To%20Point))
  (bind AttributeMathsOp_28 (Attribute%20Maths%20Op :InputSource1 $Position.Z :InputSource2 $Position.Z :Operation Set :OutputTarget $Position.Z))
  (bind GetActorProperty_8 (Get%20Actor%20Property :PropertyName Edge%20Width))
  (bind Distance_0 (Distance :bSetDensity true :MaximumDistance 512.0))
  (bind AttributeMathsOp_11 (Attribute%20Maths%20Op :InputSource1 $Density :Operation OneMinus))
  (bind GetActorProperty_9 (Get%20Actor%20Property :PropertyName Edge%20Width))
  (bind AttributeMathsOp_40 (Attribute%20Maths%20Op :Operation Multiply))
  (bind AddAttribute_10 (Add%20Attribute :AttributeTypes (struct :Type Vector) :OutputTarget SplineVector))
  (bind AddAttribute_9 (Add%20Attribute :AttributeTypes (struct :Type Vector) :OutputTarget SplineVector))
  (bind AttributeMathsOp_32 (Attribute%20Maths%20Op :InputSource1 @Last :InputSource2 $Rotation.Right :Operation Set :OutputTarget SplineVector))
  (bind Projection_4 (Projection :bKeepZeroDensityPoints true :ProjectionParams (struct :AttributeList "SplineVector" :AttributeMode IncludeAttributes)))
  (bind AttractElement_0 (Attract))
  (bind AttributeMathsOp_41 (Attribute%20Maths%20Op :InputSource1 $Position :InputSource2 $Position :Operation Set :OutputTarget $Position))
  (bind AttributeMathsOp_2 (Attribute%20Maths%20Op :InputSource1 $Position :InputSource2 $Position :Operation Lerp :OutputTarget $Position))
  (bind AttributeMathsOp_42 (Attribute%20Maths%20Op :InputSource1 @Last :InputSource2 SplineVector :Operation Set :OutputTarget SplineVector))
  (bind AddAttribute_12 (Add%20Attribute :AttributeTypes (struct :Type Float) :OutputTarget Distance))
  (bind GetActorProperty_11 (Get%20Actor%20Property :PropertyName Layer))
  (bind GetActorProperty_12 (Get%20Actor%20Property :PropertyName Priority))
  (bind MegaMeshWrite_1 (MeshPartition%20Write :Priority 1e+02 :Type None)) ; Here we write back the altered mesh partition points \nand update any desired weight channel as well
  (bind AttributeMathsOp_43 (Attribute%20Maths%20Op :Operation Multiply))
  (bind Distance_2 (Distance :bSetDensity true :MaximumDistance 512.0))
  (bind AttributeMathsOp_44 (Attribute%20Maths%20Op :InputSource1 Distance :Operation Divide :OutputTarget Distance))
  (bind AttributeMathsOp_46 (Attribute%20Maths%20Op :Operation Subtract))
  (bind AttributeMathsOp_47 (Attribute%20Maths%20Op :Operation Abs))
  (bind AttributeMathsOp_49 (Attribute%20Maths%20Op :Operation Clamp))
  (bind AttributeMathsOp_48 (Attribute%20Maths%20Op :Operation Multiply))
  (bind AttributeMathsOp_50 (Attribute%20Maths%20Op :InputSource1 Distance :Operation OneMinus))
  (bind CopyAttributes_6 (Copy%20Attributes :InputSource @Last :OutputTarget $Density))
  (bind AttributeMathsOp_51 (Attribute%20Maths%20Op :InputSource2 Distance :Operation Set :OutputTarget Distance))
  (bind AttributeMathsOp_53 (Attribute%20Maths%20Op :Operation Multiply))
  (bind AttributeMathsOp_37 (Attribute%20Maths%20Op :InputSource1 SplineVector :Operation Multiply :OutputTarget SplineVector))
  (bind AttributeMathsOp_54 (Attribute%20Maths%20Op :InputSource1 Distance :InputSource2 Distance :Operation Multiply :OutputTarget Distance))
  (bind AttributeMathsOp_52 (Attribute%20Maths%20Op :InputSource1 SplineVector :InputSource2 Distance :Operation Multiply :OutputTarget SplineVector))
  (bind AttributeMathsOp_38 (Attribute%20Maths%20Op :InputSource1 SplineVector :InputSource2 $Position :OutputTarget $Position))
  (bind GetActorProperty_1 (Get%20Actor%20Property :PropertyName Offset))
  ;[0.15 0.15 0.15 0.5] Generate Polygon for Top Mask
  (bind GetActorData_2 (Get%20Actor%20Data :bMergeSinglePointData true :Mode GetSinglePoint))
  (bind CreatePolygon2D_0 (Create%20Polygon%202D :OpenPolygonWidth 128.0))
  (bind CreateSurfaceFromPolygon2D_1 (Create%20Surface%20From%20Polygon2D))
  (bind ToPoint_0 (To%20Point))
  (bind AttributeMathsOp_29 (Attribute%20Maths%20Op :InputSource1 $Position.Z :InputSource2 $Position.Z :Operation Set :OutputTarget $Position.Z))
  ;[0.15 0.15 0.15 0.5] Add Noise to Spline
  (bind GetActorProperty_5 (Get%20Actor%20Property :PropertyName Procedural%20Noise%20Tiling))
  (bind MakeTransformAttribute_0 (Make%20Transform%20Attribute))
  (bind FilterDataByType_2 (Filter%20Data%20By%20Type :TargetType (struct :Ids [(struct :Struct /Script/PCG.PCGDataTypeInfoParam)] :CustomSubtype 8)))
  (bind Spatial%20Noise_2 (Spatial%20Noise :Iterations 8 :Transform (struct :Scale3D (32.0 32.0 32.0))))
  (bind AttributeMathsOp_22 (Attribute%20Maths%20Op :InputSource1 $Density :Operation Subtract :OutputTarget $Density))
  (bind AddAttribute_3 (Add%20Attribute :AttributeTypes (struct :Type Vector) :OutputTarget NoiseVector))
  (bind CopyAttributes_9 (Copy%20Attributes :InputSource $Rotation.Right :OutputTarget NoiseVector))
  (bind GetActorProperty_2 (Get%20Actor%20Property :PropertyName Displacement%20Amount))
  (bind AttributeMathsOp_23 (Attribute%20Maths%20Op :Operation Multiply))
  (bind AttributeMathsOp_25 (Attribute%20Maths%20Op :Operation Multiply))
  (bind AttributeMathsOp_26 (Attribute%20Maths%20Op :InputSource1 $Position :InputSource2 NoiseVector))
  ;[0.15 0.15 0.15 0.5] Calculate Interior Mask for SDF
  (bind GetActorProperty_10 (Get%20Actor%20Property :PropertyName Edge%20Width))
  (bind Distance_1 (Distance :bSetDensity true :MaximumDistance 512.0))
  (bind AttributeMathsOp_7 (Attribute%20Maths%20Op :Operation Subtract :OutputTarget $Density))
  (bind AttributeMathsOp_9 (Attribute%20Maths%20Op :Operation Floor))
  (bind AttributeMathsOp_8 (Attribute%20Maths%20Op :Operation Clamp))
  (bind AttributeMathsOp_27 (Attribute%20Maths%20Op :Operation Divide))
  ;[0.15 0.15 0.15 0.5] Apply Curve Asset Fallof
  (bind AttributeMathsOp_4 (Attribute%20Maths%20Op :InputSource1 $Density :Operation OneMinus))
  (bind AttributeRemap_0 (Attribute%20Remap :Curve (struct :ExternalCurve /Hera/Environment/Terrain/MeshTerrain/Assets/C_Smooth.C_Smooth) :Mode Curve))
  (bind AttributeMathsOp_13 (Attribute%20Maths%20Op :Operation Clamp))
  ;[0.15 0.15 0.15 0.5] Combine Masks
  (bind AttributeMathsOp_12 (Attribute%20Maths%20Op :Operation Max))
  ;[0.15 0.15 0.15 0.5] Dilate to Expand Spline Shape to Project onto
  (bind AddAttribute_4 (Add%20Attribute :AttributeTypes (struct :Type Vector) :OutputTarget OffsetVector))
  (bind GetActorProperty_6 (Get%20Actor%20Property :PropertyName Edge%20Width))
  (bind AttributeMathsOp_10 (Attribute%20Maths%20Op :Operation Multiply))
  (bind AttributeMathsOp_6 (Attribute%20Maths%20Op :InputSource1 $Rotation.Right :Operation Multiply :OutputTarget OffsetVector))
  (bind AttributeMathsOp_0 (Attribute%20Maths%20Op :InputSource1 $Position :InputSource2 OffsetVector))
  (bind FilterDataByType_0 (Filter%20Data%20By%20Type :TargetType (struct :Ids [(struct :Struct /Script/PCG.PCGDataTypeInfoPoint) (struct :Struct /Script/PCG.PCGDataTypeInfoSpline)])))
  (bind CreatePolygon2D_1 (Create%20Polygon%202D :OpenPolygonWidth 128.0))
  (bind CreateSurfaceFromPolygon2D_0 (Create%20Surface%20From%20Polygon2D))
  (connect Spatial%20Noise_2 Out AddAttribute_3 In)
  (connect SplineSampler_1 Out AddAttribute_4 In)
  (connect SplineSampler_1 Out AddAttribute_7 In)
  (connect CopyAttributes_3 Out AddAttribute_8 In)
  (connect CreatePolygon2D_1 Out AddAttribute_9 In)
  (connect AttributeMathsOp_5 Out AddAttribute_10 In)
  (connect CopyAttributes_1 Out AddAttribute_11 In)
  (connect AttributeMathsOp_42 Out AddAttribute_12 In)
  (connect AttributeMathsOp_40 Out AttractElement_0 Distance)
  (connect Projection_4 Out AttractElement_0 Source)
  (connect Select_0 Out AttractElement_0 Target)
  (connect AttributeMathsOp_6 Out AttributeMathsOp_0 InA)
  (connect AttributeMathsOp_6 Out AttributeMathsOp_0 InB)
  (connect GetActorProperty_0 Out AttributeMathsOp_1 InA)
  (connect AddAttribute_11 Out AttributeMathsOp_2 InA)
  (connect AttributeMathsOp_41 Out AttributeMathsOp_2 InB)
  (connect AttributeMathsOp_12 Out AttributeMathsOp_2 Ratio)
  (connect GetActorProperty_4 Out AttributeMathsOp_3 InA)
  (connect Distance_0 Out AttributeMathsOp_4 In)
  (connect CopyAttributes_1 Out AttributeMathsOp_5 InA)
  (connect GetActorData_0 Out AttributeMathsOp_5 InB)
  (connect AddAttribute_4 Out AttributeMathsOp_6 InA)
  (connect AttributeMathsOp_10 Out AttributeMathsOp_6 InB)
  (connect Distance_1 Out AttributeMathsOp_7 InA)
  (connect AttributeMathsOp_9 Out AttributeMathsOp_8 In)
  (connect AttributeMathsOp_7 Out AttributeMathsOp_9 In)
  (connect GetActorProperty_6 Out AttributeMathsOp_10 InA)
  (connect AttributeMathsOp_8 Out AttributeMathsOp_11 In)
  (connect AttributeMathsOp_13 Out AttributeMathsOp_12 InA)
  (connect AttributeMathsOp_11 Out AttributeMathsOp_12 InB)
  (connect AttributeRemap_0 Out AttributeMathsOp_13 In)
  (connect AddAttribute_8 Out AttributeMathsOp_14 InA)
  (connect AttributeMathsOp_3 Out AttributeMathsOp_14 InB)
  (connect AddAttribute_8 Out AttributeMathsOp_15 InA)
  (connect AttributeMathsOp_17 Out AttributeMathsOp_15 InB)
  (connect AttributeMathsOp_21 Out AttributeMathsOp_17 InA)
  (connect AttributeMathsOp_21 Out AttributeMathsOp_17 InB)
  (connect SampleTexture_1 Out AttributeMathsOp_19 InA)
  (connect GetActorProperty_3 Out AttributeMathsOp_19 InB)
  (connect AttributeMathsOp_30 Out AttributeMathsOp_21 InA)
  (connect AttributeMathsOp_31 Out AttributeMathsOp_21 InB)
  (connect CustomHLSL_1 Outpoints AttributeMathsOp_21 Ratio)
  (connect Spatial%20Noise_2 Out AttributeMathsOp_22 InA)
  (connect CopyAttributes_9 Out AttributeMathsOp_23 InA)
  (connect GetActorProperty_2 Out AttributeMathsOp_23 InB)
  (connect SampleTexture_2 Out AttributeMathsOp_24 InA)
  (connect GetActorProperty_3 Out AttributeMathsOp_24 InB)
  (connect AttributeMathsOp_23 Out AttributeMathsOp_25 InA)
  (connect AttributeMathsOp_22 Out AttributeMathsOp_25 InB)
  (connect AttributeMathsOp_25 Out AttributeMathsOp_26 InA)
  (connect AttributeMathsOp_25 Out AttributeMathsOp_26 InB)
  (connect GetActorProperty_10 Out AttributeMathsOp_27 InA)
  (connect ToPoint_2 Out AttributeMathsOp_28 InA)
  (connect GetActorData_1 Out AttributeMathsOp_28 InB)
  (connect ToPoint_0 Out AttributeMathsOp_29 InA)
  (connect GetActorData_2 Out AttributeMathsOp_29 InB)
  (connect AttributeMathsOp_19 Out AttributeMathsOp_30 InA)
  (connect AttributeMathsOp_1 Out AttributeMathsOp_30 InB)
  (connect AttributeMathsOp_24 Out AttributeMathsOp_31 InA)
  (connect AttributeMathsOp_1 Out AttributeMathsOp_31 InB)
  (connect AddAttribute_9 Out AttributeMathsOp_32 InA)
  (connect AddAttribute_9 Out AttributeMathsOp_32 InB)
  (connect AttributeMathsOp_51 Out AttributeMathsOp_37 InA)
  (connect AttributeMathsOp_53 Out AttributeMathsOp_37 InB)
  (connect AttributeMathsOp_52 Out AttributeMathsOp_38 InA)
  (connect AttributeMathsOp_52 Out AttributeMathsOp_38 InB)
  (connect GetActorProperty_9 Out AttributeMathsOp_40 InA)
  (connect AttractElement_0 Out AttributeMathsOp_41 InA)
  (connect Projection_4 Out AttributeMathsOp_41 InB)
  (connect AttributeMathsOp_2 Out AttributeMathsOp_42 InA)
  (connect AttributeMathsOp_41 Out AttributeMathsOp_42 InB)
  (connect GetActorProperty_8 Out AttributeMathsOp_43 InA)
  (connect Distance_2 Out AttributeMathsOp_44 InA)
  (connect AttributeMathsOp_43 Out AttributeMathsOp_44 InB)
  (connect AttributeMathsOp_44 Out AttributeMathsOp_46 InA)
  (connect AttributeMathsOp_46 Out AttributeMathsOp_47 In)
  (connect AttributeMathsOp_49 Out AttributeMathsOp_48 InA)
  (connect AttributeMathsOp_47 Out AttributeMathsOp_49 In)
  (connect AttributeMathsOp_48 Out AttributeMathsOp_50 In)
  (connect AddAttribute_12 Out AttributeMathsOp_51 InA)
  (connect CopyAttributes_6 Out AttributeMathsOp_51 InB)
  (connect AttributeMathsOp_54 Out AttributeMathsOp_52 InA)
  (connect AttributeMathsOp_54 Out AttributeMathsOp_52 InB)
  (connect AttributeMathsOp_8 Out AttributeMathsOp_53 InB)
  (connect AttributeMathsOp_37 Out AttributeMathsOp_54 InA)
  (connect AttributeMathsOp_37 Out AttributeMathsOp_54 InB)
  (connect AttributeMathsOp_4 Out AttributeRemap_0 In)
  (connect ToPoint_1 Out CopyAttributes_1 Source)
  (connect ToPoint_1 Out CopyAttributes_1 Target)
  (connect AddAttribute_7 Out CopyAttributes_3 Source)
  (connect AddAttribute_7 Out CopyAttributes_3 Target)
  (connect AttributeMathsOp_50 Out CopyAttributes_6 Source)
  (connect AttributeMathsOp_50 Out CopyAttributes_6 Target)
  (connect AddAttribute_3 Out CopyAttributes_9 Source)
  (connect AddAttribute_3 Out CopyAttributes_9 Target)
  (connect Select_0 Out CreatePolygon2D_0 In)
  (connect FilterDataByType_0 InsideFilter CreatePolygon2D_1 In)
  (connect CreatePolygon2D_1 Out CreateSurfaceFromPolygon2D_0 In)
  (connect CreatePolygon2D_0 Out CreateSurfaceFromPolygon2D_1 In)
  (connect SampleTexture_1 Out CustomHLSL_1 Points)
  (connect GetActorProperty_8 Out Distance_0 MaximumDistance)
  (connect AttributeMathsOp_5 Out Distance_0 Source)
  (connect AttributeMathsOp_28 Out Distance_0 Target)
  (connect GetActorProperty_10 Out Distance_1 MaximumDistance)
  (connect Distance_0 Out Distance_1 Source)
  (connect AttributeMathsOp_29 Out Distance_1 Target)
  (connect AttributeMathsOp_43 Out Distance_2 MaximumDistance)
  (connect AttributeMathsOp_5 Out Distance_2 Source)
  (connect AttributeMathsOp_28 Out Distance_2 Target)
  (connect AttributeMathsOp_0 Out FilterDataByType_0 In)
  (connect GetSplineData_0 Out FilterDataByType_1 In)
  (connect MakeTransformAttribute_0 Out FilterDataByType_2 In)
  (connect GetActorProperty_5 Out MakeTransformAttribute_0 Scale)
  (connect AddAttribute_12 Out MegaMeshWrite_1 In)
  (connect GetActorProperty_12 Out MegaMeshWrite_1 Priority)
  (connect GetActorProperty_11 Out MegaMeshWrite_1 Type)
  (connect AddAttribute_10 Out Projection_4 In)
  (connect AttributeMathsOp_32 Out Projection_4 Projection%20Target)
  (connect GetTextureData_2 Out SampleTexture_1 BaseTexture)
  (connect AttributeMathsOp_14 Out SampleTexture_1 Point)
  (connect GetTextureData_2 Out SampleTexture_2 BaseTexture)
  (connect AttributeMathsOp_14 Out SampleTexture_2 Point)
  (connect GetActorProperty_7 Out Select_0 bUseInputB)
  (connect AttributeMathsOp_26 Out Select_0 Input%20A)
  (connect AttributeMathsOp_15 Out Select_0 Input%20B)
  (connect SplineSampler_1 Out Spatial%20Noise_2 In)
  (connect FilterDataByType_2 InsideFilter Spatial%20Noise_2 Transform)
  (connect FilterDataByType_1 InsideFilter SplineDirection_0 In)
  (connect SplineDirection_0 Out SplineSampler_1 Spline)
  (connect CreateSurfaceFromPolygon2D_1 Out ToPoint_0 In)
  (connect MegaMeshQuery_1 Out ToPoint_1 In)
  (connect Select_0 Out ToPoint_2 In))
