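; PCG graph that samples the physical material of the surface hit by a raycast from the PCG component and displays it through a spawned actor.
; One created point feeds an unbounded complex-trace raycast, whose PhysicalMaterial is copied back onto that point.
; An attached BP_Sample_PhysMat_Text actor spawns there with its PhysicalMaterial property overridden.
; A parallel math chain turns the point-to-hit height into a Z offset and scale for a debug shape.
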(graph PCG_Sample_PhysMat
  (bind CreatePoints_0 (Create%20Points :CoordinateSpace OriginalComponent :PointsToCreate [(struct :BoundsMin (-1e+01 -1e+01 -1e+01) :BoundsMax (1e+01 1e+01 1e+01) :Steepness 1.0)]))
  (bind WorldRaycast_0 (World%20Raycast :bUnbounded true :WorldQueryParams (struct :bTraceComplex true :bGetReferenceToPhysicalMaterial true)))
  (bind AttributeMathsOp_0 (Attribute%20Maths%20Op :InputSource1 $Position.Z :InputSource2 $Position.Z :Operation Subtract :OutputTarget Height))
  (bind AttributeMathsOp_1 (Attribute%20Maths%20Op :InputSource1 Height :Operation Divide :OutputTarget Height))
  (bind AttributeMathsOp_2 (Attribute%20Maths%20Op :InputSource1 $Position.Z :InputSource2 Height :Operation Subtract :OutputTarget $Position.Z))
  (bind AttributeMathsOp_4 (Attribute%20Maths%20Op :InputSource1 Height :Operation Divide :OutputTarget Height))
  (bind AttributeMathsOp_3 (Attribute%20Maths%20Op :InputSource1 $Scale.Z :InputSource2 Height :Operation Set :OutputTarget $Scale.Z))
  (bind Debug_0 (Debug))
  (bind CopyAttributes_0 (Copy%20Attributes :InputSource PhysicalMaterial :OutputTarget PhysicalMaterial))
  (bind SpawnActor_1 (Spawn%20Actor :AttachOptions Attached :bAllowTemplateActorEditing true :SpawnedActorPropertyOverrideDescriptions [(struct :InputSource PhysicalMaterial :PropertyTarget "PhysicalMaterial")] :TemplateActor (class:/Hera/Environment/Terrain/MeshTerrain/Assets/BP_Sample_PhysMat_Text.BP_Sample_PhysMat_Text_C) :TemplateActorClass /Hera/Environment/Terrain/MeshTerrain/Assets/BP_Sample_PhysMat_Text.BP_Sample_PhysMat_Text_C))
  (connect CreatePoints_0 Out AttributeMathsOp_0 InA)
  (connect WorldRaycast_0 Out AttributeMathsOp_0 InB)
  (connect AttributeMathsOp_0 Out AttributeMathsOp_1 InA)
  (connect AttributeMathsOp_1 Out AttributeMathsOp_2 InA)
  (connect AttributeMathsOp_1 Out AttributeMathsOp_2 InB)
  (connect AttributeMathsOp_4 Out AttributeMathsOp_3 InA)
  (connect AttributeMathsOp_4 Out AttributeMathsOp_3 InB)
  (connect AttributeMathsOp_2 Out AttributeMathsOp_4 InA)
  (connect WorldRaycast_0 Out CopyAttributes_0 Source)
  (connect CreatePoints_0 Out CopyAttributes_0 Target)
  (connect AttributeMathsOp_3 Out Debug_0 In)
  (connect CopyAttributes_0 Out SpawnActor_1 In)
  (connect CreatePoints_0 Out WorldRaycast_0 Origins))
